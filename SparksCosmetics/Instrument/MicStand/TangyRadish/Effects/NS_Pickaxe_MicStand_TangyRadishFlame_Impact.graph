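# Niagara impact effect for the TangyRadish Flame mic stand pickaxe, a pink-tinted one-shot burst of flash, smoke, sparkle and debris.
# Core_Impact layers sprites, mesh and light through spawn-group masks, alongside confetti, glow drops, smoke and glitter emitters.
# PhysMatDebris picks debris behavior from User.PhysicalSurface, with its collision module disabled.
# System GlobalScale, time-of-day color and curves, and front-end preview switches adjust size, brightness and motion.
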NiagaraSystem /SparksCosmetics/Instrument/MicStand/TangyRadish/Effects/NS_Pickaxe_MicStand_TangyRadishFlame_Impact
dependencies {
 usedDataInterfaces [
  &NiagaraDataInterfaceCurve
  &NiagaraDataInterfaceColorCurve
  &NiagaraDataInterfaceArrayInt32
  &NiagaraDataInterfaceSpriteRendererInfo
  &NiagaraDataInterfaceVectorCurve
  &NiagaraDataInterfaceArrayColor
  &NiagaraDataInterfaceMeshRendererInfo
  &NiagaraDataInterfaceArrayFloat
  &NiagaraDataInterfaceVector2DCurve
 ]
 usedDynamicInputs []
 usedModules [
  &self:SystemSpawnScript.SystemScriptSource.SystemScriptGraph.NiagaraNodeAssignment_0.SetVariables_#1
  &SystemState
 ]
}
emitters {
 Confetti {
  emitterUpdateScript [
   &EmitterState()
   &SpawnBurst_Instantaneous(
    Spawn Count=&RandomRangeInt(Minimum=5, Maximum=7)
    Use Spawn Probability=false
    UseLoopCountLimit=false
   )
  ]
  particleSpawnScript [
   /SetVariables(Particles.VisibilityTag=&SpawnGroup()) {isSetParameters=true}
   &InitializeParticle(
    Color=&ColorFromCurve(
     ColorCurve=DataInterface() {propertyValues {
       AlphaCurve {keys [<at 0 1 RCIM_Cubic>]}
       BlueCurve {keys [<at 0 0.9384>,<at 0.4349 0.8366>,<at 0.9986 1>]}
       GreenCurve {keys [<at 0 0.7969>,<at 0.4349 0.9703>,<at 0.9986 0.8958>]}
       RedCurve {keys [<at 0 1>,<at 0.4349 1>,<at 0.9986 0.8476>]}
      }}
     CurveIndex=&RandomRangeFloat(Minimum=0, Maximum=1)
    )
    Lifetime=&FloatFromCurve(
     FloatCurve=DataInterface() {propertyValues {Curve {keys [<at 0.3 0.6>,<at 0.9 0.7>,<at 1 1>]}}}
     CurveIndex=&RandomRangeFloat(Minimum=0, Maximum=1)
    )
    Sprite Size Mode=::NewEnumerator1
    Uniform Sprite Size Max=20
    Uniform Sprite Size Min=10
    UsePositionOffset=false
    Write Material Random=true
   )
   /SetVariables(Particles.RandomCore:NiagaraFloat=&FloatFromCurve(
     FloatCurve=DataInterface() {propertyValues {Curve {keys [
         {arriveTangent=0.0559, interpMode=RCIM_Cubic, leaveTangent=0.0559, tangentMode=RCTM_User, time=0, value=0.15}
         {arriveTangent=0.1895, interpMode=RCIM_Cubic, leaveTangent=0.1895, tangentMode=RCTM_User, time=0.9, value=0.3}
         {arriveTangent=1.1808, interpMode=RCIM_Cubic, leaveTangent=1.1808, tangentMode=RCTM_User, time=1, value=0.4}
        ]}}}
     CurveIndex=&RandomRangeFloat(Minimum=0, Maximum=1)
     Scale Curve=2
    )) {isSetParameters=true}
   &ShapeLocation(MaskSpawnGroup=false, Override Seed=false, Sphere Radius=5, UseOwnerScale=true)
   &AddVelocityFromPoint(
    Module.Use Falloff Distance=false
    Origin Offset=<0 0 -1>
    Velocity Strength=&Multiply_Float(A=&RandomRangeFloat(Minimum=150, Maximum=400), B=$System.GlobalScale)
   )
   &AddVelocity(
    Velocity=&RandomRangeVector(Minimum=<0 -25 0>, Maximum=<0 -45 0>)
    Velocity Speed Scale=&FN_FrontEndFloatModifier(Gameplay=0.75, Front End=0)
   )
   &ApplyInitialForces()
   &CameraOffset(Camera Offset Amount=&FN_FrontEndFloatModifier(Gameplay=15, Front End=10))
  ]
  particleUpdateScript [
   &ParticleState()
   &ScaleSpriteSize(
    Scale Sprite Size Mode=::NewEnumerator0
    Uniform Scale Factor=&FloatFromCurve(
     FloatCurve=DataInterface() {propertyValues {Curve {keys [
         {arriveTangent=0.0271, interpMode=RCIM_Cubic, leaveTangent=0.0271, leaveTangentWeight=0.3333, tangentMode=RCTM_User, time=0, value=0}
         {
          arriveTangent=-0.0952
          arriveTangentWeight=0.0667
          interpMode=RCIM_Cubic
          leaveTangent=-0.0952
          leaveTangentWeight=0.2667
          tangentMode=RCTM_User
          time=0.2
          value=0.3973
         }
         <at 1 0 RCIM_Cubic>
        ]}}}
     Scale Curve=$System.GlobalScale
    )
   ) {module=ScaleSpriteSize001}
   &SpriteRotationRate(Rotation Rate=&Multiply_FloatByInt(
     Float=&RandomRangeFloat(Minimum=15, Maximum=30)
     Integer=&SelectIntFromArray(
      Array Sampling Mode=::NewEnumerator1
      Direct Array Index=?
      Int Selection Array=DataInterface() {propertyValues {IntData [-1,1]}}
      Random Seed=0
     )
    ))
   &CurlNoiseForce(
    Noise Frequency=4
    Noise Strength=&FloatFromCurve(
     FloatCurve=DataInterface() {propertyValues {Curve {
        defaultValue=1
        keys [
         {arriveTangent=1.3351, interpMode=RCIM_Cubic, leaveTangent=1.3351, tangentMode=RCTM_User, time=0, value=0}
         {interpMode=RCIM_Cubic, tangentMode=RCTM_User, time=1, value=1}
        ]
       }}}
     Scale Curve=100
    )
    Pan Noise=true
    Pan Noise Field=<0.5 0.5 0>
   )
   &SubUVAnimation(
    "Sprite Renderer "=DataInterface() {propertyValues {SpriteRenderer=&self:Confetti_0.NiagaraSpriteRendererProperties_0}}
    Start Frame Offset:NiagaraInt32=&RandomRangeInt(Minimum=0, Maximum=36)
    UseEndFrame=false
    UseStartFrame=false
   ) {module=SubUVAnimation001}
   &GravityForce(Gravity=&RandomRangeVector(Minimum=<0 0 -250>, Maximum=<1 1 -500>)) {module=GravityForce001}
   &Drag(Drag=&RandomRangeFloat(Minimum=3.25, Maximum=4), Use Linear Drag=true, Use Rotational Drag=true) {module=Drag001}
   &ScaleColor(
    Scale RGB=&Multiply_VectorByFloat(
     Vector=&Multiply_Vector(A=<1 1 1>, B=&MakeVectorFromLinearColorRGB(Color=<rgba 1 1 1 1>))
     Float=&FN_FrontEndFloatModifier(Gameplay=1, Front End=1.75)
    )
    ScaleA=true
    ScaleRGB=true
   )
   &SolveForcesAndVelocity(Clamp Velocity=false, Limit Acceleration=false)
  ]
  rendererClasses [&NiagaraSpriteRendererProperties]
  renderers [{rendererClass=&NiagaraSpriteRendererProperties}]
 }
 Core_Impact {
  emitterUpdateScript [
   &EmitterState()
   &SpawnBurst_Instantaneous(Use Spawn Probability=false, UseLoopCountLimit=false)
   &SpawnBurst_Instantaneous(Spawn Group=1, Use Spawn Probability=false, UseLoopCountLimit=false) {module=SpawnBurst_Instantaneous001}
   &SpawnBurst_Instantaneous(Spawn Group=2, Use Spawn Probability=false, UseLoopCountLimit=false) {module=SpawnBurst_Instantaneous002}
   &SpawnBurst_Instantaneous(Spawn Group=3, Use Spawn Probability=false, UseLoopCountLimit=false) {module=SpawnBurst_Instantaneous003}
   &SpawnBurst_Instantaneous(Spawn Count=9, Spawn Group=4, Use Spawn Probability=false) {module=SpawnBurst_Instantaneous004}
  ]
  particleSpawnScript [
   /SetVariables(Particles.VisibilityTag=&SpawnGroup()) {isSetParameters=true}
   &InitializeParticle(
    Color=&MaskLinearColorBySpawnGroup(
     Masked Linear Color=<rgba 1 1 1 1>
     Passthrough Linear Color=&MaskLinearColorBySpawnGroup(
      Masked Linear Color=&SelectLinearColorFromArray(
       Array Sampling Mode=::NewEnumerator0
       Direct Array Index=?
       Color Selection Array=DataInterface() {propertyValues {ColorData [<rgba 0.6849 0.0423 0.1279 1>, <rgba 0.8724 0.2794 0.4079 1>]}}
       Random Seed=0
      )
      Passthrough Linear Color=<rgba 0.3115 0.0391 0.107 0.7>
      Spawn Group Mask=4
     )
     Spawn Group Mask=2
    )
    Lifetime=&MaskFloatBySpawnGroup(
     Masked Float=0.15
     Passthrough Float=&MaskFloatBySpawnGroup(
      Masked Float=0.3
      Passthrough Float=&MaskFloatBySpawnGroup(Masked Float=&RandomRangeFloat(Minimum=0.15, Maximum=0.25), Passthrough Float=0.2, Spawn Group Mask=4)
      Spawn Group Mask=3
     )
     Spawn Group Mask=2
    )
    Mesh Scale Mode=::NewEnumerator3
    Mesh Uniform Scale=0.35
    Sprite Rotation Mode=::NewEnumerator2
    Sprite Size=&MaskVector2dBySpawnGroup(
     Masked Vector2D=&Vector2DFromFloat(Value=&Multiply_Float(A=120, B=$System.GlobalScale))
     Passthrough Vector2D=&MaskVector2dBySpawnGroup(
      Masked Vector2D=&RandomRangeVector2D(Minimum=<9 50>, Maximum=<15 60>)
      Passthrough Vector2D=&Vector2DFromFloat(Value=150)
      Spawn Group Mask=4
     )
     Spawn Group Mask=2
    )
    Sprite Size Mode=::NewEnumerator4
    UsePositionOffset=false
    Write Material Random=true
   )
   &FN_PerParticleLOD(Enable SpriteSize Distance Scaling=true, Sprite Size Scale Far=<2.5 2.5>)
   /SetVariables(
    Particles.Flash Star Cam Offset=20
    Particles.Flash Star Lifetime=0.25
    Particles.Flash Star Sprite Size Alt:Vector2f=&Vector2DFromFloat(Value=87)
    Particles.Flash Star SubUV Index Alt:NiagaraFloat=&RandomRangeFloat(Minimum=1, Maximum=3)
   ) {isSetParameters=true}
   /SetVariables(
    Particles.FlareCardColor:LinearColor=$System.TOD Color
    Particles.FlareCardDynamicParameter:Vector4f=&MakeVector4(X=0.2, Y=1, Z=0, W=0)
   ) {isSetParameters=true}
   &ShapeLocation(
    MaskSpawnGroup=false
    Sphere Distribution=::NewEnumerator3
    Sphere Radius=&MaskFloatBySpawnGroup(Masked Float=0.5, Passthrough Float=0, Spawn Group Mask=4)
    Uniform Spiral Amount=0.8
    UseOwnerScale=true
   )
   &AddVelocityFromPoint(
    Module.Use Falloff Distance=false
    Velocity Strength=&MaskFloatBySpawnGroup(
     Masked Float=&FloatFromCurve(
      FloatCurve=DataInterface() {propertyValues {Curve {keys [<at 0 250>,<at 0.3 200.4118>,<at 0.4961 85.1765>,<at 1 38.2353>]}}}
      CurveIndex=&RandomRangeFloat(Minimum=0, Maximum=1)
      Scale Curve=$System.GlobalScale
     )
     Passthrough Float=0
     Spawn Group Mask=4
    )
   )
   &InitialMeshOrientation(Orientation Axis=<0 0 1>, Orientation Vector=<0 0 1>, Rotation=<0 0.25 0>, Use Orientation Vector=true, Use Rotation Vector=true)
   &InitialMeshOrientation(
    Orientation Axis=<0 0 0>
    Orientation Vector=<0 0 0>
    Rotation=&RandomRangeVector(Minimum=<0 0 0>, Maximum=<1 0 0>)
    Use Orientation Vector=false
    Use Rotation Vector=true
   ) {module=InitialMeshOrientation001}
   &CameraOffset(Camera Offset Amount=&MaskFloatBySpawnGroup(
     Masked Float=15
     Passthrough Float=&MaskFloatBySpawnGroup(
      Masked Float=13
      Passthrough Float=&MaskFloatBySpawnGroup(Masked Float=15, Passthrough Float=5, Spawn Group Mask=4)
      Spawn Group Mask=3
     )
     Spawn Group Mask=2
    ))
  ]
  particleUpdateScript [
   &ParticleState()
   &ScaleSpriteSize(
    Scale Factor=&MaskVector2dBySpawnGroup(
     Masked Vector2D=&Multiply_Vector2D_ByFloat(
      Float=&FloatFromCurve(FloatCurve=DataInterface() {propertyValues {Curve {
          defaultValue=1
          keys [
           {interpMode=RCIM_Cubic, tangentMode=RCTM_User, time=0, value=1}
           {arriveTangent=-3.0168, interpMode=RCIM_Cubic, leaveTangent=-3.0168, tangentMode=RCTM_User, time=1, value=0}
          ]
         }}})
      Vector2D=<1 1>
     )
     Passthrough Vector2D=&MaskVector2dBySpawnGroup(
      Masked Vector2D=&Vector2DFromFloat(Value=&FloatFromCurve(
        FloatCurve=DataInterface() {propertyValues {Curve {
           defaultValue=0.6
           keys [
            {arriveTangent=4.0885, interpMode=RCIM_Cubic, leaveTangent=4.0885, tangentMode=RCTM_User, time=0, value=0}
            {arriveTangent=0.3429, interpMode=RCIM_Cubic, leaveTangent=0.3429, tangentMode=RCTM_User, time=0.4, value=0.8618}
            {arriveTangent=0.4463, interpMode=RCIM_Cubic, leaveTangent=0.4463, tangentMode=RCTM_User, time=1, value=1}
           ]
          }}}
        Scale Curve=&Multiply_Float(A=0.75, B=$System.GlobalScale)
       ))
      Passthrough Vector2D=&Vector2DFromFloat(Value=&FloatFromCurve(
        FloatCurve=DataInterface() {propertyValues {Curve {
           defaultValue=0
           keys [
            {arriveTangent=2.4894, interpMode=RCIM_Cubic, leaveTangent=2.4894, tangentMode=RCTM_User, time=0, value=0}
            {interpMode=RCIM_Cubic, tangentMode=RCTM_User, time=1, value=1}
           ]
          }}}
        Scale Curve=$System.GlobalScale
       ))
      Spawn Group Mask=4
     )
     Spawn Group Mask=2
    )
    Scale Sprite Size Mode=::NewEnumerator2
   )
   &ScaleMeshSize(Scale Factor=&VectorFromFloat(Value=&FloatFromCurve(
      FloatCurve=DataInterface() {propertyValues {Curve {keys [
          {arriveTangent=6.4516, interpMode=RCIM_Cubic, leaveTangent=6.4516, tangentMode=RCTM_User, time=0, value=0}
          {arriveTangent=1.9764, interpMode=RCIM_Cubic, leaveTangent=1.9764, tangentMode=RCTM_User, time=0.3898, value=1.8059}
          {arriveTangent=0.1291, interpMode=RCIM_Cubic, leaveTangent=0.1291, tangentMode=RCTM_User, time=1, value=2.5}
         ]}}}
      Scale Curve=$System.GlobalScale
     )))
   /SetVariables(
    Particles.Dynamic002:Vector4f=&MakeVector4(
     X=&FloatFromCurve(FloatCurve=DataInterface() {propertyValues {Curve {
         defaultValue=1
         keys [
          {arriveTangent=-2.1953, interpMode=RCIM_Cubic, leaveTangent=-2.1953, tangentMode=RCTM_User, time=0, value=1.5}
          {interpMode=RCIM_Cubic, tangentMode=RCTM_User, time=1, value=0}
         ]
        }}})
     Y=0
     Z=0
     W=0
    )
    Particles.Dynamic1:Vector4f=<0 0.5 1 1>
   ) {isSetParameters=true}
   /SetVariables(Particles.WispyFlareColor:LinearColor=&MakeLinearColorFromVectorAndFloat(
     "Vector (RGB)"=&Multiply_Vector(
      A=&MakeVectorFromLinearColorRGB(Color=$System.TOD Color)
      B=&MakeVectorFromLinearColorRGB(Color=&ColorFromCurve(ColorCurve=DataInterface() {propertyValues {
          AlphaCurve {keys [<at 0 1 RCIM_Cubic>,<at 1 0 RCIM_Cubic>]}
          BlueCurve {keys [<at 0 0.5036 RCIM_Cubic>,<at 0.9341 0.4685 RCIM_Cubic>]}
          GreenCurve {keys [<at 0 0.4297 RCIM_Cubic>,<at 0.9341 0.2359 RCIM_Cubic>]}
          RedCurve {keys [<at 0 1 RCIM_Cubic>,<at 0.9341 0.9245 RCIM_Cubic>]}
         }}))
     )
     "Float (Alpha)"=&FloatFromCurve(
      FloatCurve=DataInterface() {propertyValues {Curve {keys [
          {interpMode=RCIM_Cubic, leaveTangentWeight=0.464, tangentMode=RCTM_User, time=0, value=2}
          {arriveTangent=-3.1122, interpMode=RCIM_Cubic, leaveTangent=-3.1122, tangentMode=RCTM_User, time=0.5, value=0.6}
          <at 1 0 RCIM_Cubic>
         ]}}}
      Scale Curve=&Multiply_Float(
       A=&MakeCustomFloatFromBool(Bool=$User.bIsFrontEndPreview, True Float=0.3, False Float=1)
       B=&FloatFromCurve(
        FloatCurve=DataInterface() {propertyValues {Curve {keys [
            {interpMode=RCIM_Cubic, leaveTangentWeight=0.1947, tangentMode=RCTM_User, time=0, value=0.9}
            <at 0.4958 1.25 RCIM_Cubic>
            {arriveTangentWeight=0.1715, interpMode=RCIM_Cubic, tangentMode=RCTM_User, time=1, value=0.9}
           ]}}}
        CurveIndex=Linked() {linkedVariable {name=NPC.FortniteNPC.TimeOfDay_0-1, type=NiagaraFloat}}
       )
      )
     )
    )) {isSetParameters=true}
   /SetVariables(Particles.Update Flash Star Age:NiagaraFloat=&Divide_Float(A=$Particles.Age, B=$Particles.Initial.Flash Star Lifetime)) {isSetParameters=true}
   /SetVariables(
    Particles.Update Flash Star Color:LinearColor=&ColorFromCurve(
     ColorCurve=DataInterface() {propertyValues {
       AlphaCurve {keys [<at 0.2805 1>,<at 1 0>]}
       BlueCurve {keys [<at 0.1682 0.9375>,<at 0.4223 0.9832>,<at 0.8398 0.3532>]}
       GreenCurve {keys [<at 0.1682 3.1153>,<at 0.4223 0.9266>,<at 0.8398 0.1798>]}
       RedCurve {keys [<at 0.1682 5>,<at 0.4223 2>,<at 0.8398 0.8958>]}
      }}
     CurveIndex=$Particles.Update Flash Star Age
    )
    Particles.Update Flash Star Sprize Size Alt:Vector2f=&Multiply_Vector2D_ByFloat(
     Float=&FloatFromCurve(
      FloatCurve=DataInterface() {propertyValues {Curve {
         defaultValue=0.2
         keys [
          {arriveTangent=3.1639, interpMode=RCIM_Cubic, leaveTangent=3.1639, tangentMode=RCTM_User, time=0, value=0}
          {arriveTangent=0.2867, arriveTangentWeight=0.4894, interpMode=RCIM_Cubic, leaveTangent=0.2867, tangentMode=RCTM_User, time=1, value=1}
         ]
        }}}
      CurveIndex=$Particles.Update Flash Star Age
      Scale Curve=$System.GlobalScale
     )
     Vector2D=$Particles.Initial.Flash Star Sprite Size Alt
    )
   ) {isSetParameters=true}
   /SetVariables(Particles.UpdateFlareCardColor:LinearColor=&MakeLinearColorFromVectorAndFloat(
     "Vector (RGB)"=&Multiply_Vector(
      A=&Multiply_Vector(
       A=&MakeVectorFromLinearColorRGB(Color=$Particles.Initial.FlareCardColor)
       B=&MakeVectorFromLinearColorRGB(Color=&ColorFromCurve(ColorCurve=DataInterface() {propertyValues {
           AlphaCurve {keys [<at 0 1 RCIM_Cubic>,<at 1 0 RCIM_Cubic>]}
           BlueCurve {keys [<at 0 0.2969>,<at 1 0.3428>]}
           GreenCurve {keys [<at 0 1.1653>,<at 1 0.3428>]}
           RedCurve {keys [<at 0 2>,<at 1 1.5>]}
          }}))
      )
      B=&VectorFromFloat(Value=&FN_FrontEndFloatModifier(Gameplay=1.5, Front End=1))
     )
     "Float (Alpha)"=&FloatFromCurve(
      FloatCurve=DataInterface() {propertyValues {Curve {keys [{arriveTangent=-1.9727, interpMode=RCIM_Cubic, time=0, value=1}, <at 1 0 RCIM_Cubic>]}}}
      Scale Curve=&MakeCustomFloatFromBool(Bool=$User.bIsFrontEndPreview, True Float=0.5, False Float=1)
     )
    )) {isSetParameters=true}
   &FN_Light_Attributes(
    Color=<rgba 1 0.33 0.4892 1>
    Create Color Attribute=true
    Create Exponent Attribute=true
    Create Position Attribute=true
    Create Radius Attribute=true
    Exponent=4
    Position=<0 5 0>
    Radius=90
    Scale Color=&FloatFromCurve(
     FloatCurve=DataInterface() {propertyValues {Curve {keys [
         {interpMode=RCIM_Cubic, tangentMode=RCTM_User, time=0, value=1}
         {arriveTangent=-0.5426, interpMode=RCIM_Cubic, leaveTangent=-0.5426, tangentMode=RCTM_User, time=0.8, value=0}
        ]}}}
     Scale Curve=10
    )
   )
   &ScaleColor(
    Scale Alpha=&MaskFloatBySpawnGroup(
     Masked Float=&FloatFromCurve(
      FloatCurve=DataInterface() {propertyValues {Curve {
         defaultValue=1
         keys [
          {arriveTangent=0.0627, interpMode=RCIM_Cubic, leaveTangent=0.0627, tangentMode=RCTM_User, time=0.1098, value=0.9764}
          {arriveTangent=-1.7566, interpMode=RCIM_Cubic, leaveTangent=-1.7566, tangentMode=RCTM_User, time=1, value=0}
         ]
        }}}
      Scale Curve=1.1
     )
     Passthrough Float=&FloatFromCurve(
      FloatCurve=DataInterface() {propertyValues {Curve {
         defaultValue=1
         keys [
          {interpMode=RCIM_Cubic, tangentMode=RCTM_User, time=0, value=1}
          {arriveTangent=-2.0745, interpMode=RCIM_Cubic, leaveTangent=-2.0745, tangentMode=RCTM_User, time=1, value=0}
         ]
        }}}
      Scale Curve=&FloatFromCurve(
       FloatCurve=DataInterface() {propertyValues {Curve {
          defaultValue=1
          keys [
           {interpMode=RCIM_Cubic, tangentMode=RCTM_User, time=0, value=0.45}
           {interpMode=RCIM_Cubic, tangentMode=RCTM_User, time=0.5, value=1}
           {tangentMode=RCTM_User, time=1, value=0.45}
          ]
         }}}
       CurveIndex=Linked() {linkedVariable {name=NPC.FortniteNPC.TimeOfDay_0-1, type=NiagaraFloat}}
      )
     )
     Spawn Group Mask=4
    )
    Scale RGB=&MaskVectorBySpawnGroup(
     Masked Vector=&MakeVectorFromLinearColorRGB(Color=$System.TOD Color)
     Passthrough Vector=<1 1 1>
     Spawn Group Mask=4
    )
    ScaleA=true
    ScaleRGB=true
   )
   &Drag(Use Linear Drag=true, Use Rotational Drag=true)
   &SolveForcesAndVelocity(Clamp Velocity=false, Limit Acceleration=false)
  ]
  rendererClasses [
   &NiagaraSpriteRendererProperties
   &NiagaraLightRendererProperties
   &NiagaraMeshRendererProperties
  ]
  renderers [
   {rendererClass=&NiagaraSpriteRendererProperties}
   {rendererClass=&NiagaraLightRendererProperties, rendererIndex=1}
   {rendererClass=&NiagaraSpriteRendererProperties, rendererIndex=2}
   {rendererClass=&NiagaraSpriteRendererProperties, rendererIndex=3}
   {rendererClass=&NiagaraMeshRendererProperties, rendererIndex=4}
   {rendererClass=&NiagaraSpriteRendererProperties, rendererIndex=5}
  ]
 }
 FN_PhysMatDebris {
  emitterUpdateScript [
   &EmitterState()
   &SpawnBurst_Instantaneous(
    Spawn Count=&RandomRangeInt(Minimum=5, Maximum=10)
    Use Spawn Probability=false
    UseLoopCountLimit=false
   )
  ]
  particleSpawnScript [
   /SetVariables(Particles.BiasedRandom=&FloatFromCurve(
     FloatCurve=DataInterface() {propertyValues {Curve {keys [
         <at 0.2 1 RCIM_Cubic>
         {arriveTangent=-3.3394, interpMode=RCIM_Cubic, leaveTangent=-3.3394, tangentMode=RCTM_User, time=0.5, value=0.5}
         <at 0.8 0 RCIM_Cubic>
        ]}}}
     CurveIndex=&RandomRangeFloat(Minimum=0, Maximum=1)
    )) {isSetParameters=true}
   &InitializeParticle(
    Color Mode=::NewEnumerator0
    Lifetime=&RandomRangeFloat(Minimum=0.5, Maximum=1.25)
    Mass=&Lerp_Float(A=0.6, B=1.2, Alpha=$Particles.BiasedRandom)
    Mass Mode=::NewEnumerator1
    Mesh Scale=&Multiply_VectorByFloat(
     Vector=&RandomRangeVector(Minimum=<0.9 0.4 0.5>, Maximum=<1 1 1>)
     Float=&Lerp_Float(A=0.7, B=1, Alpha=$Particles.BiasedRandom)
    )
    Mesh Scale Mode=::NewEnumerator4
    UsePositionOffset=false
    Write Material Random=true
   )
   &FN_PhysicalSurface(
    Dirt Mesh Count=2
    Glass Mesh Count=3
    Ice Mesh Count=2
    Metal Gravity=&RandomRangeFloat(Minimum=-350, Maximum=-1800)
    Metal Mesh Count=2
    Metal Velocity=&RandomRangeFloat(Minimum=80, Maximum=375)
    Metal Velocity Cone Angle=130
    Physical Surface=$User.PhysicalSurface
    Stone Gravity=&RandomRangeFloat(Minimum=-700, Maximum=-1250)
    Stone Mesh Count=2
    Stone Velocity=&RandomRangeFloat(Minimum=80, Maximum=550)
    Stone Velocity Cone Angle=120
    Wood Gravity=&RandomRangeFloat(Minimum=-450, Maximum=-900)
    Wood Mesh Count=2
    Wood Scale=1.5
    Wood Velocity=&RandomRangeFloat(Minimum=80, Maximum=600)
    Wood Velocity Cone Angle=130
   )
   &AddVelocityInCone(
    Cone Angle=$Output.FN_PhysicalSurface.SurfaceResponseConeAngle
    Module.Use Velocity Falloff On Cone Axis=false
    Velocity Strength=$Output.FN_PhysicalSurface.SurfaceResponseVelocity
   )
   &InitialMeshOrientation(Use Rotation Vector=false)
   &AddRotationalVelocity(Rotation Rate=&RandomVector(Vector Scale=&RandomRangeFloat(Minimum=1, Maximum=2)))
   &ApplyInitialForces()
   /SetVariables(Particles.GravityOverrideValue=$Output.FN_PhysicalSurface.GravityOverride) {isSetParameters=true}
  ]
  particleUpdateScript [
   &ParticleState()
   &GravityForce(Gravity=$Particles.GravityOverrideValue)
   &Collision(
    Advanced Aging Rate=1
    Age Colliding Particles=false
    CPU Collision Trace Channel=ECC_WorldStatic
    Collision Enabled=&Boolean_AndOperation(
     Bool1=&SetBoolByFloatComparison(A=$Particles.Initial.Lifetime, B=1)
     Bool2=&SetBoolByFloatComparison(A=$Particles.Age, B=0.12)
    )
    Control Roll On Collision=true
    Friction=1
    Friction During a Bounce=0.25
    Max CPU Trace Length=100
    Module.Control Rest State Via Surface Normal=false
    Module.EnableMaxCollisionCount=false
    Module.Randomize Collision Normal Vector=true
    Randomize Collision Normal=0.25
    ScaleRateOfRotation=false
   ) {enabled=false}
   &SolveForcesAndVelocity(Clamp Velocity=false, Limit Acceleration=false)
   &ScaleMeshSize(Scale Factor=&VectorFromFloat(Value=&FloatFromCurve(FloatCurve=DataInterface() {propertyValues {Curve {keys [<at 0.6 1 RCIM_Cubic>, {arriveTangent=-6.308, interpMode=RCIM_Cubic, leaveTangent=-6.308, tangentMode=RCTM_User, time=1, value=0}]}}})))
   &DynamicMaterialParameters(Module.Param0WriteEnabled=true, Module.Param1WriteEnabled=true, Module.Param2WriteEnabled=false, Module.Param3WriteEnabled=false)
  ]
  rendererClasses [&NiagaraSpriteRendererProperties, &NiagaraMeshRendererProperties]
  renderers [
   {rendererClass=&NiagaraSpriteRendererProperties}
   {rendererClass=&NiagaraMeshRendererProperties, rendererIndex=1}
   {rendererClass=&NiagaraMeshRendererProperties, rendererIndex=2}
   {rendererClass=&NiagaraMeshRendererProperties, rendererIndex=3}
   {rendererClass=&NiagaraMeshRendererProperties, rendererIndex=4}
   {rendererClass=&NiagaraMeshRendererProperties, rendererIndex=5}
   {rendererClass=&NiagaraMeshRendererProperties, rendererIndex=6}
   {rendererClass=&NiagaraMeshRendererProperties, rendererIndex=7}
   {rendererClass=&NiagaraMeshRendererProperties, rendererIndex=8}
  ]
 }
 GlowDrops {
  emitterUpdateScript [
   &EmitterState(
    Enable Distance Culling=true
    Max Distance Response=::KillImmediately
    Module.bApplySpawnCountScaleByDistance=true
    Module.bUseMaxDistance=true
    Module.bUseMinDistance=false
    Scalability Mode=::NewEnumerator1
    Scale Spawn Count=true
    Spawn Count Scale By Distance=DataInterface() {propertyValues {Curve {keys [<at 0.3676 0.9657>,<at 0.9962 0.4982>]}}}
   )
   &SpawnBurst_Instantaneous(
    Spawn Count=&UniformRangedInt(Minimum=10, Maximum=15)
    Use Spawn Probability=false
    UseLoopCountLimit=false
   )
  ]
  particleSpawnScript [
   &PartitionParticles(Number Of Partitions=3)
   &InitializeParticle(
    Color=&SelectLinearColorFromArray(
     Array Sampling Mode=::NewEnumerator1
     Direct Array Index=$Particles.PartitionParticles.Partition
     Color Selection Array=DataInterface() {propertyValues {ColorData [<rgba 1 0.3297 0.3727 1>, <rgba 1 0.2179 0.3232 1>, <rgba 2 1.0375 0.4844 1>]}}
     Random Seed=?
    )
    Lifetime=&FloatFromCurve(
     FloatCurve=DataInterface() {propertyValues {Curve {keys [<at 0.15 0.6>, {arriveTangent=-0.9801, time=0.4, value=0.45}, {arriveTangent=-0.2795, leaveTangent=-0.2795, time=0.8, value=0.3}]}}}
     CurveIndex=&RandomRangeFloat(Minimum=0, Maximum=1)
    )
    Sprite Size Mode=::NewEnumerator3
    Uniform Sprite Size=&Multiply_Float(
     A=&RandomRangeFloat(Minimum=5.5, Maximum=6)
     B=&MaskFloat(
      Mask Source=1
      Mask Parameter=$Particles.PartitionParticles.Partition
      Mask=2
      Masked Float=0.85
      Passthrough Float=&RandomRangeFloat(Minimum=1.2, Maximum=1)
     )
    )
    UsePositionOffset=false
    Write Material Random=true
   )
   &SphereLocation(
    MaskSpawnGroup=false
    Override Seed=false
    Sphere Radius=5
    Surface Only=false
    UseNonUniformScale=false
    UseOffset=false
    UseSphereOrientation=false
   ) {module=SphereLocation001}
   &AddVelocityFromPoint(
    Module.Use Falloff Distance=false
    Velocity Strength=&Multiply_Float(
     A=&RandomRangeFloat(Minimum=500, Maximum=700)
     B=&Multiply_Float(
      A=&SelectFloatFromArray(Direct Array Index=$Particles.PartitionParticles.Partition, Float Selection Array=DataInterface() {propertyValues {FloatData [1,0.9,0.55]}})
      B=$System.GlobalScale
     )
    )
   )
  ]
  particleUpdateScript [
   &ParticleState()
   &GravityForce(Gravity=&Multiply_VectorByFloat(
     Vector=<0 0 -1>
     Float=&FloatFromCurve(
      FloatCurve=DataInterface() {propertyValues {Curve {
         defaultValue=1
         keys [
          {interpMode=RCIM_Cubic, leaveTangentWeight=0.4106, tangentMode=RCTM_User, time=0, value=0.5}
          {arriveTangent=0.9376, interpMode=RCIM_Cubic, leaveTangent=0.9376, tangentMode=RCTM_User, time=1, value=1}
         ]
        }}}
      Scale Curve=450
     )
    ))
   &Drag(Drag=&UniformRangedFloat(Minimum=4, Maximum=9), Use Linear Drag=true, Use Rotational Drag=false)
   &SolveForcesAndVelocity(Clamp Velocity=false, Limit Acceleration=false)
   &ScaleSpriteSize(Scale Factor=&Vector2DFromCurve(
     Vector2Curve=DataInterface() {propertyValues {
       XCurve {keys [
         {arriveTangent=0.0804, interpMode=RCIM_Cubic, leaveTangent=0.0804, tangentMode=RCTM_User, time=0, value=0.1}
         {arriveTangent=-0.1, interpMode=RCIM_Cubic, leaveTangent=-0.1, tangentMode=RCTM_User, time=0.241, value=0.6552}
         {interpMode=RCIM_Cubic, tangentMode=RCTM_User, time=1, value=0}
        ]}
       YCurve {keys [
         {arriveTangent=0.2693, interpMode=RCIM_Cubic, leaveTangent=0.2693, tangentMode=RCTM_User, time=0, value=0.1}
         {arriveTangent=-1.8506, interpMode=RCIM_Cubic, leaveTangent=-1.8506, tangentMode=RCTM_User, time=0.3912, value=0.7339}
         {interpMode=RCIM_Cubic, tangentMode=RCTM_User, time=0.9988, value=0}
        ]}
      }}
     ScaleCurve=&Multiply_Vector2D_ByFloat(Float=$System.GlobalScale, Vector2D=<1 1>)
    ))
   &ScaleSpriteSizeBySpeed(Max Scale Factor=<0.5 8>, Min Scale Factor=<0.8 1.5>)
   &ScaleColor(
    Scale Alpha=&FloatFromCurve(FloatCurve=DataInterface() {propertyValues {Curve {keys [<at 0.4 1 RCIM_Cubic>,<at 1 0>]}}})
    Scale RGB=&MaskVector(
     Mask Source=1
     Mask Parameter=$Particles.PartitionParticles.Partition
     Mask=2
     Masked Vector=&MakeVectorFromLinearColorRGB(Color=&ColorFromCurve(ColorCurve=DataInterface() {propertyValues {
         AlphaCurve {keys []}
         BlueCurve {keys [<at 0 2>,<at 0.8524 0.2987>,<at 1 0.0541>]}
         GreenCurve {keys [<at 0 2>,<at 0.8524 0>,<at 1 0>]}
         RedCurve {keys [<at 0 2>,<at 0.8524 1.0155>,<at 1 0.1875>]}
        }}))
     Passthrough Vector=&MakeVectorFromLinearColorRGB(Color=&ColorFromCurve(ColorCurve=DataInterface() {propertyValues {AlphaCurve {keys []}, BlueCurve {keys [<at 0 2>,<at 1 0.5>]}, GreenCurve {keys [<at 0 2>,<at 1 0.5>]}, RedCurve {keys [<at 0 2>,<at 1 0.5>]}}}))
    )
    ScaleA=true
    ScaleRGB=true
   )
  ]
  rendererClasses [&NiagaraSpriteRendererProperties]
  renderers [{rendererClass=&NiagaraSpriteRendererProperties}]
 }
 Impact_Smoke {
  emitterUpdateScript [
   &EmitterState()
   &SpawnBurst_Instantaneous(
    Spawn Count=&RandomRangeInt(Minimum=2, Maximum=3)
    Use Spawn Probability=false
    UseLoopCountLimit=false
   )
  ]
  particleSpawnScript [
   &InitializeParticle(
    Color Maximum=<rgba 0.79 0.4079 0.516 0.1>
    Color Minimum=<rgba 0.8841 0.4708 0.5414 0.2>
    Color Mode=::NewEnumerator2
    Lifetime Max=0.4
    Lifetime Min=0.25
    Lifetime Mode=::NewEnumerator1
    Sprite Size Mode=::NewEnumerator3
    Uniform Sprite Size=&Multiply_Float(A=$System.GlobalScale, B=&RandomRangeFloat(Minimum=100, Maximum=130))
    UsePositionOffset=false
    Write Material Random=true
   )
   &CameraOffset(Camera Offset Amount=&RandomRangeFloat(Minimum=40, Maximum=50, Evaluation Type=::NewEnumerator1))
   &ShapeLocation(
    Hemisphere Distribution=<1 0.5>
    MaskSpawnGroup=false
    Override Seed=false
    Rotation Angle=90
    Rotation Mode=::NewEnumerator1
    Sphere Radius=&MakeCustomFloatFromBool(Bool=$User.bIsFrontEndPreview, True Float=5, False Float=10)
    UseOwnerScale=true
   )
   &AddVelocity(
    Velocity=&Multiply_VectorByFloat(Vector=$Particles.ShapeLocation.ShapeVector, Float=10)
    Velocity Speed Scale=$System.GlobalScale
   )
  ]
  particleUpdateScript [
   &ParticleState()
   &ScaleSpriteSize(Uniform Curve Sprite Scale=DataInterface() {propertyValues {Curve {keys [
        {arriveTangent=3.4424, interpMode=RCIM_Cubic, leaveTangent=3.4424, tangentMode=RCTM_User, time=-0.002, value=0.1471}
        {arriveTangent=0.2424, interpMode=RCIM_Cubic, leaveTangent=0.2424, tangentMode=RCTM_User, time=0.6971, value=0.7912}
        {arriveTangent=0.86, interpMode=RCIM_Cubic, leaveTangent=0.86, tangentMode=RCTM_User, time=1, value=1}
       ]}}})
   &ScaleColor(
    Scale Alpha=&FloatFromCurve(
     FloatCurve=DataInterface() {propertyValues {Curve {keys [
         {arriveTangent=2.8076, interpMode=RCIM_Cubic, leaveTangent=2.8076, tangentMode=RCTM_User, time=0, value=0}
         {arriveTangentWeight=0.0948, interpMode=RCIM_Cubic, leaveTangentWeight=0.2845, time=0.25, value=0.5}
         {arriveTangent=-0.5429, interpMode=RCIM_Cubic, leaveTangent=-0.5429, tangentMode=RCTM_User, time=1, value=0}
        ]}}}
     Scale Curve=&FN_FrontEndFloatModifier(Gameplay=1, Front End=0.5)
    )
    Scale RGB=&VectorFromCurve(VectorCurve=DataInterface() {propertyValues {XCurve {keys [<at 0 1>,<at 1 1>]}, YCurve {keys [<at 0 1>,<at 1 1>]}, ZCurve {keys [<at 0 1>,<at 1 1>]}}})
    ScaleA=true
    ScaleRGB=false
   )
   &DynamicMaterialParameters(
    Index 0 Param 1=&Add_Float(
     A=&RandomRangeFloat(Minimum=0, Maximum=1)
     B=&FloatFromCurve(
      FloatCurve=DataInterface() {propertyValues {Curve {keys [
          {arriveTangent=1.4109, interpMode=RCIM_Cubic, leaveTangent=1.4109, tangentMode=RCTM_User, time=0, value=0}
          {arriveTangent=0.8637, interpMode=RCIM_Cubic, leaveTangent=0.8637, tangentMode=RCTM_User, time=0.2557, value=0.5257}
          {arriveTangent=1.3248, interpMode=RCIM_Cubic, leaveTangent=1.3248, tangentMode=RCTM_User, time=1, value=1}
         ]}}}
      CurveIndex=$Particles.Age
      Scale Curve=0.25
     )
    )
    Module.Param0WriteEnabled=true
    Module.Param1WriteEnabled=true
    Module.Param2WriteEnabled=true
    Module.Param3WriteEnabled=true
   ) {module=DynamicMaterialParameters001}
   &SolveForcesAndVelocity(Clamp Velocity=false, Limit Acceleration=false)
  ]
  rendererClasses [&NiagaraSpriteRendererProperties]
  renderers [
   {rendererClass=&NiagaraSpriteRendererProperties}
   {rendererClass=&NiagaraSpriteRendererProperties, rendererIndex=1}
  ]
 }
 glitter {
  emitterUpdateScript [
   &EmitterState(Life Cycle Mode=::NewEnumerator1, Loop Behavior=::NewEnumerator1, Loop Delay=0.1, Loop Duration=1, UseLoopDelay=false)
   &SpawnBurst_Instantaneous(Spawn Count=4, Use Spawn Probability=false)
  ]
  particleSpawnScript [
   &InitializeParticle(
    Color Maximum=<rgba 2 0.7812 1.1638 1>
    Color Minimum=<rgba 1.5 1.3413 0.7563 0.95>
    Color Mode=::NewEnumerator2
    Lifetime=&RandomRangeFloat(Minimum=0.35, Maximum=0.55)
    Mesh Scale Mode=::NewEnumerator3
    Mesh Uniform Scale=0.35
    Sprite Rotation Mode=::NewEnumerator2
    Sprite Size Mode=::NewEnumerator3
    Uniform Sprite Size=&RandomRangeFloat(Minimum=140, Maximum=165)
    UsePositionOffset=false
    Write Material Random=true
   )
   &CameraOffset(Camera Offset Amount=&MakeCustomFloatFromBool(Bool=$User.bIsFrontEndPreview, True Float=1, False Float=15))
   &ShapeLocation(MaskSpawnGroup=false, Sphere Distribution=::NewEnumerator3, Sphere Radius=1, UseOwnerScale=true)
   /SetVariables(Particles.VisibilityTag=&MakeCustomIntFromBool(Bool=$User.bIsFrontEndPreview, True INT=1, False INT=0)) {isSetParameters=true}
  ]
  particleUpdateScript [
   &ParticleState()
   &SpriteRotationRate(Rotation Rate=&Multiply_FloatByInt(
     Float=&Multiply_Float(
      A=&RandomRangeFloat(Minimum=20, Maximum=80)
      B=&FloatFromCurve(FloatCurve=DataInterface() {propertyValues {Curve {
          defaultValue=1
          keys [
           {arriveTangent=-2.1783, interpMode=RCIM_Cubic, leaveTangent=-2.1783, tangentMode=RCTM_User, time=0.2, value=1}
           {arriveTangentWeight=0.7172, interpMode=RCIM_Cubic, tangentMode=RCTM_User, time=1, value=0}
          ]
         }}})
     )
     Integer=&RandomRangeInt(Minimum=-1, Maximum=1)
    ))
   &ScaleSpriteSize(Scale Factor=&Multiply_Vector2D_ByFloat(
     Float=&FloatFromCurve(
      FloatCurve=DataInterface() {propertyValues {Curve {
         defaultValue=1
         keys [
          {arriveTangent=5.0192, interpMode=RCIM_Cubic, leaveTangent=5.0192, tangentMode=RCTM_User, time=0, value=0}
          {
           arriveTangent=0.4561
           arriveTangentWeight=0.1819
           interpMode=RCIM_Cubic
           leaveTangent=0.4561
           leaveTangentWeight=0.5128
           tangentMode=RCTM_User
           time=0.4
           value=0.7749
          }
          {arriveTangent=-0.0535, interpMode=RCIM_Cubic, leaveTangent=-0.0535, tangentMode=RCTM_User, time=0.9971, value=0.924}
         ]
        }}}
      Scale Curve=$System.GlobalScale
     )
     Vector2D=<1 1>
    ))
   &ScaleColor(
    Scale Alpha=&FloatFromCurve(
     FloatCurve=DataInterface() {propertyValues {Curve {
        defaultValue=1
        keys [
         {arriveTangent=1.507, interpMode=RCIM_Cubic, leaveTangent=1.507, tangentMode=RCTM_User, time=0, value=0.1}
         {arriveTangent=-2.9536, interpMode=RCIM_Cubic, leaveTangent=-2.9536, tangentMode=RCTM_User, time=0.15, value=1}
         {arriveTangent=-0.0039, arriveTangentWeight=0.3565, interpMode=RCIM_Cubic, leaveTangent=-0.0039, tangentMode=RCTM_User, time=1, value=0}
        ]
       }}}
     Scale Curve=&Multiply_Float(
      A=&FloatFromCurve(
       FloatCurve=DataInterface() {propertyValues {Curve {keys [<at 0.2 0.75 RCIM_Cubic>,<at 0.5 1 RCIM_Cubic>,<at 1 0.75 RCIM_Cubic>]}}}
       CurveIndex=Linked() {linkedVariable {name=NPC.FortniteNPC.TimeOfDay_0-1, type=NiagaraFloat}}
      )
      B=&FN_FrontEndFloatModifier(Gameplay=1, Front End=0.2)
     )
    )
    Scale RGB=&Multiply_Vector(
     A=&MakeVectorFromLinearColorRGB(Color=&ColorFromCurve(ColorCurve=DataInterface() {propertyValues {
         AlphaCurve {keys [<at 0 1 RCIM_Cubic>]}
         BlueCurve {keys [
           <at 0 0.7677>
           {arriveTangent=-0.3302, interpMode=RCIM_Cubic, leaveTangent=-0.3302, time=0.1013, value=0.7488}
           <at 0.6416 0.5558>
           <at 1 0.6246 RCIM_Cubic>
          ]}
         GreenCurve {keys [
           <at 0 0.6407>
           {arriveTangent=-0.3151, interpMode=RCIM_Cubic, leaveTangent=-0.3151, time=0.1013, value=0.5466}
           <at 0.6416 0.4385>
           <at 1 0.3992 RCIM_Cubic>
          ]}
         RedCurve {keys [<at 0 1>,<at 0.1013 1.5 RCIM_Cubic>,<at 0.6416 1>,<at 1 1 RCIM_Cubic>]}
        }}))
     B=&MakeVectorFromLinearColorRGB(Color=$System.TOD Color)
    )
    ScaleA=true
    ScaleRGB=true
   )
   &GravityForce(Gravity=&Multiply_VectorByFloat(
     Vector=<0 0 -1>
     Float=&FloatFromCurve(
      FloatCurve=DataInterface() {propertyValues {Curve {
         defaultValue=1
         keys [
          {interpMode=RCIM_Cubic, tangentMode=RCTM_User, time=0.2, value=0}
          {arriveTangent=1.6951, interpMode=RCIM_Cubic, leaveTangent=1.6951, tangentMode=RCTM_User, time=1, value=1}
         ]
        }}}
      Scale Curve=250
     )
    ))
   &SolveForcesAndVelocity(Clamp Velocity=false, Limit Acceleration=false)
  ]
  rendererClasses [&NiagaraSpriteRendererProperties]
  renderers [
   {rendererClass=&NiagaraSpriteRendererProperties}
   {rendererClass=&NiagaraSpriteRendererProperties, rendererIndex=1}
  ]
 }
}
summary {
 systemName=NS_Pickaxe_MicStand_TangyRadishFlame_Impact
 emitters [
  {
   emitterName=FN_PhysMatDebris
   rendererClasses [&NiagaraSpriteRendererProperties, &NiagaraMeshRendererProperties]
  }
  {emitterName=Confetti, rendererClasses [&NiagaraSpriteRendererProperties]}
  {emitterName=Impact_Smoke, rendererClasses [&NiagaraSpriteRendererProperties]}
  {emitterName=GlowDrops, rendererClasses [&NiagaraSpriteRendererProperties]}
  {emitterName=glitter, rendererClasses [&NiagaraSpriteRendererProperties]}
  {
   emitterName=Core_Impact
   rendererClasses [
    &NiagaraSpriteRendererProperties
    &NiagaraLightRendererProperties
    &NiagaraMeshRendererProperties
   ]
  }
 ]
 userVariables [
  User.bIsFrontEnd {type=NiagaraBool}
  User.bIsFrontEndPreview {type=NiagaraBool}
  User.PhysicalSurface {type=NiagaraInt32}
 ]
}
