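# Physics asset for the TangyRadish Flame mic stand instrument cosmetic in SparksCosmetics.
# It holds a single body on the handle bone, approximated by one long, thin capsule rotated to lie along the stand.
# It defines no constraints, so the stand has no jointed parts.
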PhysicsAsset /SparksCosmetics/Instrument/MicStand/TangyRadish/Meshes/TangyRadish_Flame_MicStand_Physics
bodies {handle [{shapeType=Capsule, length=90, radius=2, rotation=<rot 0 -90 0>}]}
constraints []
