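# Animation Blueprint for the TangyRadish Flame mic stand cosmetic that adds procedural secondary motion to its many ribbon bones.
# The input pose feeds a long serial chain of per-bone AnimDynamics nodes with angular springs, gravity override and individually tuned angular limits.
# They simulate root-relative with damped component motion; ribbon_t_02 appears twice in the chain.
# A RigidBody node based on the handle bone has no pose input wired.
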AnimBlueprint /SparksCosmetics/Instrument/MicStand/TangyRadish/Meshes/TangyRadish_Flame_MicStand_AnimBP
target_skeleton=&/TangyRadish_Flame_MicStand_Skeleton
graphs [
 AnimGraph {nodes [
   RigidBody() {
    bOverrideWorldGravity=true
    overridePhysicsAsset=&/TangyRadish_Flame_MicStand_Physics
    simulationSpace=BaseBoneSpace
    alphaBoolBlend {blendOption=Linear}
    baseBoneRef {boneName=handle}
    componentAppliedLinearAccClamp=<8000 8000 8000>
    componentLinearAccScale=<0.6 0.6 0.6>
    overrideWorldGravity {z=-980}
   }
   n1 = Anim Dynamics - Bone: ribbon_a_01!(
    ComponentPose=Local To Component(LocalPose=Input Pose())
    bAlphaBoolEnabled:bool=True
    GravityOverride:Vector="0,0,-980"
    AngularSpringConstant:float=500
    ExternalForce="0,0,0"
   ) -> (Pose) {
    alphaInputType=Bool
    angularSpringConstant=500
    bAngularSpring=true
    bUseGravityOverride=true
    simulationSpace=RootRelative
    alphaBoolBlend {blendOption=Linear}
    boundBone {boneName=ribbon_a_01}
    componentAppliedLinearAccClamp=<500 500 500>
    componentLinearAccScale=<0.2 0.2 0.2>
    componentLinearVelScale=<0.4 0.4 0.4>
    gravityOverride {z=-980}
    physicsBodyDefinitions [{
      boundBone {boneName=ribbon_a_01}
      boxExtents=<4 4 4>
      constraintSetup {angularLimitsMax=<15 60 20>, angularLimitsMin=<-10 -90 -20>, angularTarget=<0 0 0>}
      localJointOffset=<0.1 0 0>
     }]
    relativeSpaceBone {boneName=handle}
    simSpaceSettings {maxAngularAcceleration=10, maxAngularVelocity=5, simSpaceAngularAlpha=0.2}
   }
   n2 = Anim Dynamics - Bone: ribbon_b_01!(ComponentPose=@n1, bAlphaBoolEnabled:bool=True, GravityOverride:Vector="0,0,-980", AngularSpringConstant:float=500, ExternalForce="0,0,0") -> (Pose) {
    alphaInputType=Bool
    angularSpringConstant=500
    bAngularSpring=true
    bUseGravityOverride=true
    simulationSpace=RootRelative
    alphaBoolBlend {blendOption=Linear}
    boundBone {boneName=ribbon_b_01}
    componentAppliedLinearAccClamp=<500 500 500>
    componentLinearAccScale=<0.2 0.2 0.2>
    componentLinearVelScale=<0.4 0.4 0.4>
    gravityOverride {z=-980}
    physicsBodyDefinitions [{
      boundBone {boneName=ribbon_b_01}
      boxExtents=<4 4 4>
      constraintSetup {angularLimitsMax=<15 60 20>, angularLimitsMin=<-10 -90 -20>, angularTarget=<0 0 0>}
      localJointOffset=<0.1 0 0>
     }]
    relativeSpaceBone {boneName=handle}
    simSpaceSettings {maxAngularAcceleration=10, maxAngularVelocity=5, simSpaceAngularAlpha=0.2}
   }
   n3 = Anim Dynamics - Bone: ribbon_b_02!(ComponentPose=@n2, bAlphaBoolEnabled:bool=True, GravityOverride:Vector="0,0,-980", AngularSpringConstant:float=500, ExternalForce="0,0,0") -> (Pose) {
    alphaInputType=Bool
    angularSpringConstant=500
    bAngularSpring=true
    bUseGravityOverride=true
    simulationSpace=RootRelative
    alphaBoolBlend {blendOption=Linear}
    boundBone {boneName=ribbon_b_02}
    componentAppliedLinearAccClamp=<500 500 500>
    componentLinearAccScale=<0.2 0.2 0.2>
    componentLinearVelScale=<0.4 0.4 0.4>
    gravityOverride {z=-980}
    physicsBodyDefinitions [{
      boundBone {boneName=ribbon_b_02}
      boxExtents=<4 4 4>
      constraintSetup {angularLimitsMax=<15 20 25>, angularLimitsMin=<-15 -60 -30>, angularTarget=<0 0 0>}
      localJointOffset=<0.05 0 0>
     }]
    relativeSpaceBone {boneName=handle}
    simSpaceSettings {maxAngularAcceleration=10, maxAngularVelocity=5, simSpaceAngularAlpha=0.2}
   }
   n4 = Anim Dynamics - Bone: ribbon_c_01!(ComponentPose=@n3, bAlphaBoolEnabled:bool=True, GravityOverride:Vector="0,0,-980", AngularSpringConstant:float=500, ExternalForce="0,0,0") -> (Pose) {
    alphaInputType=Bool
    angularSpringConstant=500
    bAngularSpring=true
    bUseGravityOverride=true
    simulationSpace=RootRelative
    alphaBoolBlend {blendOption=Linear}
    boundBone {boneName=ribbon_c_01}
    componentAppliedLinearAccClamp=<500 500 500>
    componentLinearAccScale=<0.2 0.2 0.2>
    componentLinearVelScale=<0.4 0.4 0.4>
    gravityOverride {z=-980}
    physicsBodyDefinitions [{
      boundBone {boneName=ribbon_c_01}
      boxExtents=<4 4 4>
      constraintSetup {angularLimitsMax=<15 60 20>, angularLimitsMin=<-10 -90 -20>, angularTarget=<0 0 0>}
      localJointOffset=<0.1 0 0>
     }]
    relativeSpaceBone {boneName=handle}
    simSpaceSettings {maxAngularAcceleration=10, maxAngularVelocity=5, simSpaceAngularAlpha=0.2}
   }
   n5 = Anim Dynamics - Bone: ribbon_c_02!(ComponentPose=@n4, bAlphaBoolEnabled:bool=True, GravityOverride:Vector="0,0,-980", AngularSpringConstant:float=500, ExternalForce="0,0,0") -> (Pose) {
    alphaInputType=Bool
    angularSpringConstant=500
    bAngularSpring=true
    bUseGravityOverride=true
    simulationSpace=RootRelative
    alphaBoolBlend {blendOption=Linear}
    boundBone {boneName=ribbon_c_02}
    componentAppliedLinearAccClamp=<500 500 500>
    componentLinearAccScale=<0.2 0.2 0.2>
    componentLinearVelScale=<0.4 0.4 0.4>
    gravityOverride {z=-980}
    physicsBodyDefinitions [{
      boundBone {boneName=ribbon_c_02}
      boxExtents=<4 4 4>
      constraintSetup {angularLimitsMax=<15 20 25>, angularLimitsMin=<-15 -50 -30>, angularTarget=<0 0 0>}
      localJointOffset=<0.05 0 0>
     }]
    relativeSpaceBone {boneName=handle}
    simSpaceSettings {maxAngularAcceleration=10, maxAngularVelocity=5, simSpaceAngularAlpha=0.2}
   }
   n6 = Anim Dynamics - Bone: ribbon_c_03!(ComponentPose=@n5, bAlphaBoolEnabled:bool=True, GravityOverride:Vector="0,0,-980", AngularSpringConstant:float=500, ExternalForce="0,0,0") -> (Pose) {
    alphaInputType=Bool
    angularSpringConstant=500
    bAngularSpring=true
    bUseGravityOverride=true
    simulationSpace=RootRelative
    alphaBoolBlend {blendOption=Linear}
    boundBone {boneName=ribbon_c_03}
    componentAppliedLinearAccClamp=<500 500 500>
    componentLinearAccScale=<0.2 0.2 0.2>
    componentLinearVelScale=<0.4 0.4 0.4>
    gravityOverride {z=-980}
    physicsBodyDefinitions [{
      boundBone {boneName=ribbon_c_03}
      boxExtents=<4 4 4>
      constraintSetup {angularLimitsMax=<15 30 45>, angularLimitsMin=<-15 -45 -40>, angularTarget=<0 0 0>}
      localJointOffset=<0.05 0 0>
     }]
    relativeSpaceBone {boneName=handle}
    simSpaceSettings {maxAngularAcceleration=10, maxAngularVelocity=5, simSpaceAngularAlpha=0.2}
   }
   n7 = Anim Dynamics - Bone: ribbon_d_01!(ComponentPose=@n6, bAlphaBoolEnabled:bool=True, GravityOverride:Vector="0,0,-980", AngularSpringConstant:float=500, ExternalForce="0,0,0") -> (Pose) {
    alphaInputType=Bool
    angularSpringConstant=500
    bAngularSpring=true
    bUseGravityOverride=true
    simulationSpace=RootRelative
    alphaBoolBlend {blendOption=Linear}
    boundBone {boneName=ribbon_d_01}
    componentAppliedLinearAccClamp=<500 500 500>
    componentLinearAccScale=<0.2 0.2 0.2>
    componentLinearVelScale=<0.4 0.4 0.4>
    gravityOverride {z=-980}
    physicsBodyDefinitions [{
      boundBone {boneName=ribbon_d_01}
      boxExtents=<4 4 4>
      constraintSetup {angularLimitsMax=<15 60 25>, angularLimitsMin=<-10 -90 -30>, angularTarget=<0 0 0>}
      localJointOffset=<0.1 0 0>
     }]
    relativeSpaceBone {boneName=handle}
    simSpaceSettings {maxAngularAcceleration=10, maxAngularVelocity=5, simSpaceAngularAlpha=0.2}
   }
   n8 = Anim Dynamics - Bone: ribbon_d_02!(ComponentPose=@n7, bAlphaBoolEnabled:bool=True, GravityOverride:Vector="0,0,-980", AngularSpringConstant:float=500, ExternalForce="0,0,0") -> (Pose) {
    alphaInputType=Bool
    angularSpringConstant=500
    bAngularSpring=true
    bUseGravityOverride=true
    simulationSpace=RootRelative
    alphaBoolBlend {blendOption=Linear}
    boundBone {boneName=ribbon_d_02}
    componentAppliedLinearAccClamp=<500 500 500>
    componentLinearAccScale=<0.2 0.2 0.2>
    componentLinearVelScale=<0.4 0.4 0.4>
    gravityOverride {z=-980}
    physicsBodyDefinitions [{
      boundBone {boneName=ribbon_d_02}
      boxExtents=<4 4 4>
      constraintSetup {angularLimitsMax=<15 20 25>, angularLimitsMin=<-15 -50 -30>, angularTarget=<0 0 0>}
      localJointOffset=<0.05 0 0>
     }]
    relativeSpaceBone {boneName=handle}
    simSpaceSettings {maxAngularAcceleration=10, maxAngularVelocity=5, simSpaceAngularAlpha=0.2}
   }
   n9 = Anim Dynamics - Bone: ribbon_e_01!(ComponentPose=@n8, bAlphaBoolEnabled:bool=True, GravityOverride:Vector="0,0,-980", AngularSpringConstant:float=500, ExternalForce="0,0,0") -> (Pose) {
    alphaInputType=Bool
    angularSpringConstant=500
    bAngularSpring=true
    bUseGravityOverride=true
    simulationSpace=RootRelative
    alphaBoolBlend {blendOption=Linear}
    boundBone {boneName=ribbon_e_01}
    componentAppliedLinearAccClamp=<500 500 500>
    componentLinearAccScale=<0.2 0.2 0.2>
    componentLinearVelScale=<0.4 0.4 0.4>
    gravityOverride {z=-980}
    physicsBodyDefinitions [{
      boundBone {boneName=ribbon_e_01}
      boxExtents=<4 4 4>
      constraintSetup {angularLimitsMax=<15 60 20>, angularLimitsMin=<-10 -90 -10>, angularTarget=<0 0 0>}
      localJointOffset=<0.1 0 0>
     }]
    relativeSpaceBone {boneName=handle}
    simSpaceSettings {maxAngularAcceleration=10, maxAngularVelocity=5, simSpaceAngularAlpha=0.2}
   }
   n10 = Anim Dynamics - Bone: ribbon_f_01!(ComponentPose=@n9, bAlphaBoolEnabled:bool=True, AngularSpringConstant:float=500, ExternalForce="0,0,0") -> (Pose) {
    alphaInputType=Bool
    angularSpringConstant=500
    bAngularSpring=true
    bUseGravityOverride=true
    simulationSpace=RootRelative
    alphaBoolBlend {blendOption=Linear}
    boundBone {boneName=ribbon_f_01}
    componentAppliedLinearAccClamp=<500 500 500>
    componentLinearAccScale=<0.2 0.2 0.2>
    componentLinearVelScale=<0.4 0.4 0.4>
    gravityOverride {z=-980}
    physicsBodyDefinitions [{
      boundBone {boneName=ribbon_f_01}
      boxExtents=<4 4 4>
      constraintSetup {angularLimitsMax=<15 45 20>, angularLimitsMin=<-10 -70 -30>, angularTarget=<0 0 0>}
      localJointOffset=<0.1 0 0>
     }]
    relativeSpaceBone {boneName=handle}
    simSpaceSettings {maxAngularAcceleration=10, maxAngularVelocity=5, simSpaceAngularAlpha=0.2}
   }
   n11 = Anim Dynamics - Bone: ribbon_f_02!(ComponentPose=@n10, bAlphaBoolEnabled:bool=True, AngularSpringConstant:float=500, ExternalForce="0,0,0") -> (Pose) {
    alphaInputType=Bool
    angularSpringConstant=500
    bAngularSpring=true
    bUseGravityOverride=true
    simulationSpace=RootRelative
    alphaBoolBlend {blendOption=Linear}
    boundBone {boneName=ribbon_f_02}
    componentAppliedLinearAccClamp=<500 500 500>
    componentLinearAccScale=<0.2 0.2 0.2>
    componentLinearVelScale=<0.4 0.4 0.4>
    gravityOverride {z=-980}
    physicsBodyDefinitions [{
      boundBone {boneName=ribbon_f_02}
      boxExtents=<4 4 4>
      constraintSetup {angularLimitsMax=<15 30 20>, angularLimitsMin=<-10 -45 -12>, angularTarget=<0 0 0>}
      localJointOffset=<0.1 0 0>
     }]
    relativeSpaceBone {boneName=handle}
    simSpaceSettings {maxAngularAcceleration=10, maxAngularVelocity=5, simSpaceAngularAlpha=0.2}
   }
   n12 = Anim Dynamics - Bone: ribbon_g_01!(ComponentPose=@n11, bAlphaBoolEnabled:bool=True, AngularSpringConstant:float=500, ExternalForce="0,0,0") -> (Pose) {
    alphaInputType=Bool
    angularSpringConstant=500
    bAngularSpring=true
    bUseGravityOverride=true
    simulationSpace=RootRelative
    alphaBoolBlend {blendOption=Linear}
    boundBone {boneName=ribbon_g_01}
    componentAppliedLinearAccClamp=<500 500 500>
    componentLinearAccScale=<0.2 0.2 0.2>
    componentLinearVelScale=<0.4 0.4 0.4>
    gravityOverride {z=-980}
    physicsBodyDefinitions [{
      boundBone {boneName=ribbon_g_01}
      boxExtents=<4 4 4>
      constraintSetup {angularLimitsMax=<15 45 40>, angularLimitsMin=<-10 -60 -10>, angularTarget=<0 0 0>}
      localJointOffset=<0.1 0 0>
     }]
    relativeSpaceBone {boneName=handle}
    simSpaceSettings {maxAngularAcceleration=10, maxAngularVelocity=5, simSpaceAngularAlpha=0.2}
   }
   n13 = Anim Dynamics - Bone: ribbon_g_02!(ComponentPose=@n12, bAlphaBoolEnabled:bool=True, AngularSpringConstant:float=500, ExternalForce="0,0,0") -> (Pose) {
    alphaInputType=Bool
    angularSpringConstant=500
    bAngularSpring=true
    bUseGravityOverride=true
    simulationSpace=RootRelative
    alphaBoolBlend {blendOption=Linear}
    boundBone {boneName=ribbon_g_02}
    componentAppliedLinearAccClamp=<500 500 500>
    componentLinearAccScale=<0.2 0.2 0.2>
    componentLinearVelScale=<0.4 0.4 0.4>
    gravityOverride {z=-980}
    physicsBodyDefinitions [{
      boundBone {boneName=ribbon_g_02}
      boxExtents=<4 4 4>
      constraintSetup {angularLimitsMax=<15 10 20>, angularLimitsMin=<-10 -45 -60>, angularTarget=<0 0 0>}
      localJointOffset=<0.1 0 0>
     }]
    relativeSpaceBone {boneName=handle}
    simSpaceSettings {maxAngularAcceleration=10, maxAngularVelocity=5, simSpaceAngularAlpha=0.2}
   }
   n14 = Anim Dynamics - Bone: ribbon_h_01!(ComponentPose=@n13, bAlphaBoolEnabled:bool=True, AngularSpringConstant:float=500, ExternalForce="0,0,0") -> (Pose) {
    alphaInputType=Bool
    angularSpringConstant=500
    bAngularSpring=true
    bUseGravityOverride=true
    simulationSpace=RootRelative
    alphaBoolBlend {blendOption=Linear}
    boundBone {boneName=ribbon_h_01}
    componentAppliedLinearAccClamp=<500 500 500>
    componentLinearAccScale=<0.2 0.2 0.2>
    componentLinearVelScale=<0.4 0.4 0.4>
    gravityOverride {z=-980}
    physicsBodyDefinitions [{
      boundBone {boneName=ribbon_h_01}
      boxExtents=<4 4 4>
      constraintSetup {angularLimitsMax=<45 50 20>, angularLimitsMin=<-30 -45 -10>, angularTarget=<0 0 0>}
      localJointOffset=<0.1 0 0>
     }]
    relativeSpaceBone {boneName=handle}
    simSpaceSettings {maxAngularAcceleration=10, maxAngularVelocity=5, simSpaceAngularAlpha=0.2}
   }
   n15 = Anim Dynamics - Bone: ribbon_h_02!(ComponentPose=@n14, bAlphaBoolEnabled:bool=True, AngularSpringConstant:float=500, ExternalForce="0,0,0") -> (Pose) {
    alphaInputType=Bool
    angularSpringConstant=500
    bAngularSpring=true
    bUseGravityOverride=true
    simulationSpace=RootRelative
    alphaBoolBlend {blendOption=Linear}
    boundBone {boneName=ribbon_h_02}
    componentAppliedLinearAccClamp=<500 500 500>
    componentLinearAccScale=<0.2 0.2 0.2>
    componentLinearVelScale=<0.4 0.4 0.4>
    gravityOverride {z=-980}
    physicsBodyDefinitions [{
      boundBone {boneName=ribbon_h_02}
      boxExtents=<4 4 4>
      constraintSetup {angularLimitsMax=<15 60 10>, angularLimitsMin=<-10 -10 -30>, angularTarget=<0 0 0>}
      localJointOffset=<0.1 0 0>
     }]
    relativeSpaceBone {boneName=handle}
    simSpaceSettings {maxAngularAcceleration=10, maxAngularVelocity=5, simSpaceAngularAlpha=0.2}
   }
   n16 = Anim Dynamics - Bone: ribbon_h_03!(ComponentPose=@n15, bAlphaBoolEnabled:bool=True, AngularSpringConstant:float=500, ExternalForce="0,0,0") -> (Pose) {
    alphaInputType=Bool
    angularSpringConstant=500
    bAngularSpring=true
    bUseGravityOverride=true
    simulationSpace=RootRelative
    alphaBoolBlend {blendOption=Linear}
    boundBone {boneName=ribbon_h_03}
    componentAppliedLinearAccClamp=<500 500 500>
    componentLinearAccScale=<0.2 0.2 0.2>
    componentLinearVelScale=<0.4 0.4 0.4>
    gravityOverride {z=-980}
    physicsBodyDefinitions [{
      boundBone {boneName=ribbon_h_03}
      boxExtents=<4 4 4>
      constraintSetup {angularLimitsMax=<15 30 60>, angularLimitsMin=<-15 -45 -45>, angularTarget=<0 0 0>}
      localJointOffset=<0.1 0 0>
     }]
    relativeSpaceBone {boneName=handle}
    simSpaceSettings {maxAngularAcceleration=10, maxAngularVelocity=5, simSpaceAngularAlpha=0.2}
   }
   n17 = Anim Dynamics - Bone: ribbon_i_01!(ComponentPose=@n16, bAlphaBoolEnabled:bool=True, AngularSpringConstant:float=500, ExternalForce="0,0,0") -> (Pose) {
    alphaInputType=Bool
    angularSpringConstant=500
    bAngularSpring=true
    bUseGravityOverride=true
    simulationSpace=RootRelative
    alphaBoolBlend {blendOption=Linear}
    boundBone {boneName=ribbon_i_01}
    componentAppliedLinearAccClamp=<500 500 500>
    componentLinearAccScale=<0.2 0.2 0.2>
    componentLinearVelScale=<0.4 0.4 0.4>
    gravityOverride {z=-980}
    physicsBodyDefinitions [{
      boundBone {boneName=ribbon_i_01}
      boxExtents=<4 4 4>
      constraintSetup {angularLimitsMax=<15 10 15>, angularLimitsMin=<-10 -70 -20>, angularTarget=<0 0 0>}
      localJointOffset=<0.1 0 0>
     }]
    relativeSpaceBone {boneName=handle}
    simSpaceSettings {maxAngularAcceleration=10, maxAngularVelocity=5, simSpaceAngularAlpha=0.2}
   }
   n18 = Anim Dynamics - Bone: ribbon_i_02!(ComponentPose=@n17, bAlphaBoolEnabled:bool=True, AngularSpringConstant:float=500, ExternalForce="0,0,0") -> (Pose) {
    alphaInputType=Bool
    angularSpringConstant=500
    bAngularSpring=true
    bUseGravityOverride=true
    simulationSpace=RootRelative
    alphaBoolBlend {blendOption=Linear}
    boundBone {boneName=ribbon_i_02}
    componentAppliedLinearAccClamp=<500 500 500>
    componentLinearAccScale=<0.2 0.2 0.2>
    componentLinearVelScale=<0.4 0.4 0.4>
    gravityOverride {z=-980}
    physicsBodyDefinitions [{
      boundBone {boneName=ribbon_i_02}
      boxExtents=<5 4 4>
      constraintSetup {angularLimitsMax=<15 10 20>, angularLimitsMin=<-10 -45 -30>, angularTarget=<0 0 0>}
      localJointOffset=<0.1 0 0>
     }]
    relativeSpaceBone {boneName=handle}
    simSpaceSettings {maxAngularAcceleration=10, maxAngularVelocity=5, simSpaceAngularAlpha=0.2}
   }
   n19 = Anim Dynamics - Bone: ribbon_j_01!(ComponentPose=@n18, bAlphaBoolEnabled:bool=True, AngularSpringConstant:float=500, ExternalForce="0,0,0") -> (Pose) {
    alphaInputType=Bool
    angularSpringConstant=500
    bAngularSpring=true
    bUseGravityOverride=true
    simulationSpace=RootRelative
    alphaBoolBlend {blendOption=Linear}
    boundBone {boneName=ribbon_j_01}
    componentAppliedLinearAccClamp=<500 500 500>
    componentLinearAccScale=<0.2 0.2 0.2>
    componentLinearVelScale=<0.4 0.4 0.4>
    gravityOverride {z=-980}
    physicsBodyDefinitions [{
      boundBone {boneName=ribbon_j_01}
      boxExtents=<4 4 4>
      constraintSetup {angularLimitsMax=<15 10 10>, angularLimitsMin=<-10 -70 -20>, angularTarget=<0 0 0>}
      localJointOffset=<0.1 0 0>
     }]
    relativeSpaceBone {boneName=handle}
    simSpaceSettings {maxAngularAcceleration=10, maxAngularVelocity=5, simSpaceAngularAlpha=0.2}
   }
   n20 = Anim Dynamics - Bone: ribbon_j_02!(ComponentPose=@n19, bAlphaBoolEnabled:bool=True, AngularSpringConstant:float=500, ExternalForce="0,0,0") -> (Pose) {
    alphaInputType=Bool
    angularSpringConstant=500
    bAngularSpring=true
    bUseGravityOverride=true
    simulationSpace=RootRelative
    alphaBoolBlend {blendOption=Linear}
    boundBone {boneName=ribbon_j_02}
    componentAppliedLinearAccClamp=<500 500 500>
    componentLinearAccScale=<0.2 0.2 0.2>
    componentLinearVelScale=<0.4 0.4 0.4>
    gravityOverride {z=-980}
    physicsBodyDefinitions [{
      boundBone {boneName=ribbon_j_02}
      boxExtents=<5 4 4>
      constraintSetup {angularLimitsMax=<15 60 20>, angularLimitsMin=<-10 -45 -30>, angularTarget=<0 0 0>}
      localJointOffset=<0.1 0 0>
     }]
    relativeSpaceBone {boneName=handle}
    simSpaceSettings {maxAngularAcceleration=10, maxAngularVelocity=5, simSpaceAngularAlpha=0.2}
   }
   n21 = Anim Dynamics - Bone: ribbon_k_01!(ComponentPose=@n20, bAlphaBoolEnabled:bool=True, AngularSpringConstant:float=500, ExternalForce="0,0,0") -> (Pose) {
    alphaInputType=Bool
    angularSpringConstant=500
    bAngularSpring=true
    bUseGravityOverride=true
    simulationSpace=RootRelative
    alphaBoolBlend {blendOption=Linear}
    boundBone {boneName=ribbon_k_01}
    componentAppliedLinearAccClamp=<500 500 500>
    componentLinearAccScale=<0.2 0.2 0.2>
    componentLinearVelScale=<0.4 0.4 0.4>
    gravityOverride {z=-980}
    physicsBodyDefinitions [{
      boundBone {boneName=ribbon_k_01}
      boxExtents=<4 4 4>
      constraintSetup {angularLimitsMax=<15 40 60>, angularLimitsMin=<-15 -60 -60>, angularTarget=<0 0 0>}
      localJointOffset=<0.05 0 0>
     }]
    relativeSpaceBone {boneName=handle}
    simSpaceSettings {maxAngularAcceleration=10, maxAngularVelocity=5, simSpaceAngularAlpha=0.2}
   }
   n22 = Anim Dynamics - Bone: ribbon_k_02!(ComponentPose=@n21, bAlphaBoolEnabled:bool=True, AngularSpringConstant:float=500, ExternalForce="0,0,0") -> (Pose) {
    alphaInputType=Bool
    angularSpringConstant=500
    bAngularSpring=true
    bUseGravityOverride=true
    simulationSpace=RootRelative
    alphaBoolBlend {blendOption=Linear}
    boundBone {boneName=ribbon_k_02}
    componentAppliedLinearAccClamp=<500 500 500>
    componentLinearAccScale=<0.2 0.2 0.2>
    componentLinearVelScale=<0.4 0.4 0.4>
    gravityOverride {z=-980}
    physicsBodyDefinitions [{
      boundBone {boneName=ribbon_k_02}
      boxExtents=<4 4 4>
      constraintSetup {angularLimitsMax=<15 10 35>, angularLimitsMin=<-10 -60 -30>, angularTarget=<0 0 0>}
      localJointOffset=<0.1 0 0>
     }]
    relativeSpaceBone {boneName=handle}
    simSpaceSettings {maxAngularAcceleration=10, maxAngularVelocity=5, simSpaceAngularAlpha=0.2}
   }
   n23 = Anim Dynamics - Bone: ribbon_k_03!(ComponentPose=@n22, bAlphaBoolEnabled:bool=True, AngularSpringConstant:float=500, ExternalForce="0,0,0") -> (Pose) {
    alphaInputType=Bool
    angularSpringConstant=500
    bAngularSpring=true
    bUseGravityOverride=true
    simulationSpace=RootRelative
    alphaBoolBlend {blendOption=Linear}
    boundBone {boneName=ribbon_k_03}
    componentAppliedLinearAccClamp=<500 500 500>
    componentLinearAccScale=<0.2 0.2 0.2>
    componentLinearVelScale=<0.4 0.4 0.4>
    gravityOverride {z=-980}
    physicsBodyDefinitions [{
      boundBone {boneName=ribbon_k_03}
      boxExtents=<4 4 4>
      constraintSetup {angularLimitsMax=<15 10 45>, angularLimitsMin=<-10 -30 -45>, angularTarget=<0 0 0>}
      localJointOffset=<0.1 0 0>
     }]
    relativeSpaceBone {boneName=handle}
    simSpaceSettings {maxAngularAcceleration=10, maxAngularVelocity=5, simSpaceAngularAlpha=0.2}
   }
   n24 = Anim Dynamics - Bone: ribbon_l_01!(ComponentPose=@n23, bAlphaBoolEnabled:bool=True, AngularSpringConstant:float=500, ExternalForce="0,0,0") -> (Pose) {
    alphaInputType=Bool
    angularSpringConstant=500
    bAngularSpring=true
    bUseGravityOverride=true
    simulationSpace=RootRelative
    alphaBoolBlend {blendOption=Linear}
    boundBone {boneName=ribbon_l_01}
    componentAppliedLinearAccClamp=<500 500 500>
    componentLinearAccScale=<0.2 0.2 0.2>
    componentLinearVelScale=<0.4 0.4 0.4>
    gravityOverride {z=-980}
    physicsBodyDefinitions [{
      boundBone {boneName=ribbon_l_01}
      boxExtents=<4 4 4>
      constraintSetup {angularLimitsMax=<15 35 60>, angularLimitsMin=<-15 -60 -60>, angularTarget=<0 0 0>}
      localJointOffset=<0.05 0 0>
     }]
    relativeSpaceBone {boneName=handle}
    simSpaceSettings {maxAngularAcceleration=10, maxAngularVelocity=5, simSpaceAngularAlpha=0.2}
   }
   n25 = Anim Dynamics - Bone: ribbon_l_02!(ComponentPose=@n24, bAlphaBoolEnabled:bool=True, AngularSpringConstant:float=500, ExternalForce="0,0,0") -> (Pose) {
    alphaInputType=Bool
    angularSpringConstant=500
    bAngularSpring=true
    bUseGravityOverride=true
    simulationSpace=RootRelative
    alphaBoolBlend {blendOption=Linear}
    boundBone {boneName=ribbon_l_02}
    componentAppliedLinearAccClamp=<500 500 500>
    componentLinearAccScale=<0.2 0.2 0.2>
    componentLinearVelScale=<0.4 0.4 0.4>
    gravityOverride {z=-980}
    physicsBodyDefinitions [{
      boundBone {boneName=ribbon_l_02}
      boxExtents=<4 4 4>
      constraintSetup {angularLimitsMax=<15 35 35>, angularLimitsMin=<-10 -60 -30>, angularTarget=<0 0 0>}
      localJointOffset=<0.1 0 0>
     }]
    relativeSpaceBone {boneName=handle}
    simSpaceSettings {maxAngularAcceleration=10, maxAngularVelocity=5, simSpaceAngularAlpha=0.2}
   }
   n26 = Anim Dynamics - Bone: ribbon_m_01!(ComponentPose=@n25, bAlphaBoolEnabled:bool=True, AngularSpringConstant:float=500, ExternalForce="0,0,0") -> (Pose) {
    alphaInputType=Bool
    angularSpringConstant=500
    bAngularSpring=true
    bUseGravityOverride=true
    simulationSpace=RootRelative
    alphaBoolBlend {blendOption=Linear}
    boundBone {boneName=ribbon_m_01}
    componentAppliedLinearAccClamp=<500 500 500>
    componentLinearAccScale=<0.2 0.2 0.2>
    componentLinearVelScale=<0.4 0.4 0.4>
    gravityOverride {z=-980}
    physicsBodyDefinitions [{
      boundBone {boneName=ribbon_m_01}
      boxExtents=<4 4 4>
      constraintSetup {angularLimitsMax=<15 30 55>, angularLimitsMin=<-15 -60 -30>, angularTarget=<0 0 0>}
      localJointOffset=<0.05 0 0>
     }]
    relativeSpaceBone {boneName=handle}
    simSpaceSettings {maxAngularAcceleration=10, maxAngularVelocity=5, simSpaceAngularAlpha=0.2}
   }
   n27 = Anim Dynamics - Bone: ribbon_m_02!(ComponentPose=@n26, bAlphaBoolEnabled:bool=True, AngularSpringConstant:float=500, ExternalForce="0,0,0") -> (Pose) {
    alphaInputType=Bool
    angularSpringConstant=500
    bAngularSpring=true
    bUseGravityOverride=true
    simulationSpace=RootRelative
    alphaBoolBlend {blendOption=Linear}
    boundBone {boneName=ribbon_m_02}
    componentAppliedLinearAccClamp=<500 500 500>
    componentLinearAccScale=<0.2 0.2 0.2>
    componentLinearVelScale=<0.4 0.4 0.4>
    gravityOverride {z=-980}
    physicsBodyDefinitions [{
      boundBone {boneName=ribbon_m_02}
      boxExtents=<4 4 4>
      constraintSetup {angularLimitsMax=<15 35 35>, angularLimitsMin=<-10 -60 -30>, angularTarget=<0 0 0>}
      localJointOffset=<0.1 0 0>
     }]
    relativeSpaceBone {boneName=handle}
    simSpaceSettings {maxAngularAcceleration=10, maxAngularVelocity=5, simSpaceAngularAlpha=0.2}
   }
   n28 = Anim Dynamics - Bone: ribbon_m_03!(ComponentPose=@n27, bAlphaBoolEnabled:bool=True, AngularSpringConstant:float=500, ExternalForce="0,0,0") -> (Pose) {
    alphaInputType=Bool
    angularSpringConstant=500
    bAngularSpring=true
    bUseGravityOverride=true
    simulationSpace=RootRelative
    alphaBoolBlend {blendOption=Linear}
    boundBone {boneName=ribbon_m_03}
    componentAppliedLinearAccClamp=<500 500 500>
    componentLinearAccScale=<0.2 0.2 0.2>
    componentLinearVelScale=<0.4 0.4 0.4>
    gravityOverride {z=-980}
    physicsBodyDefinitions [{
      boundBone {boneName=ribbon_m_03}
      boxExtents=<4 4 4>
      constraintSetup {angularLimitsMax=<15 45 45>, angularLimitsMin=<-10 -30 -45>, angularTarget=<0 0 0>}
      localJointOffset=<0.1 0 0>
     }]
    relativeSpaceBone {boneName=handle}
    simSpaceSettings {maxAngularAcceleration=10, maxAngularVelocity=5, simSpaceAngularAlpha=0.2}
   }
   n29 = Anim Dynamics - Bone: ribbon_n_01!(ComponentPose=@n28, bAlphaBoolEnabled:bool=True, AngularSpringConstant:float=500, ExternalForce="0,0,0") -> (Pose) {
    alphaInputType=Bool
    angularSpringConstant=500
    bAngularSpring=true
    bUseGravityOverride=true
    simulationSpace=RootRelative
    alphaBoolBlend {blendOption=Linear}
    boundBone {boneName=ribbon_n_01}
    componentAppliedLinearAccClamp=<500 500 500>
    componentLinearAccScale=<0.2 0.2 0.2>
    componentLinearVelScale=<0.4 0.4 0.4>
    gravityOverride {z=-980}
    physicsBodyDefinitions [{
      boundBone {boneName=ribbon_n_01}
      boxExtents=<4 4 4>
      constraintSetup {angularLimitsMax=<15 30 25>, angularLimitsMin=<-15 -60 -20>, angularTarget=<0 0 0>}
      localJointOffset=<0.1 0 0>
     }]
    relativeSpaceBone {boneName=handle}
    simSpaceSettings {maxAngularAcceleration=10, maxAngularVelocity=5, simSpaceAngularAlpha=0.2}
   }
   n30 = Anim Dynamics - Bone: ribbon_o_01!(ComponentPose=@n29, bAlphaBoolEnabled:bool=True, AngularSpringConstant:float=500, ExternalForce="0,0,0") -> (Pose) {
    alphaInputType=Bool
    angularSpringConstant=500
    bAngularSpring=true
    bUseGravityOverride=true
    simulationSpace=RootRelative
    alphaBoolBlend {blendOption=Linear}
    boundBone {boneName=ribbon_o_01}
    componentAppliedLinearAccClamp=<500 500 500>
    componentLinearAccScale=<0.2 0.2 0.2>
    componentLinearVelScale=<0.4 0.4 0.4>
    gravityOverride {z=-980}
    physicsBodyDefinitions [{
      boundBone {boneName=ribbon_o_01}
      boxExtents=<4 4 4>
      constraintSetup {angularLimitsMax=<0 35 30>, angularLimitsMin=<0 -60 -45>, angularTarget=<0 0 0>}
      localJointOffset=<0.05 0 0>
     }]
    relativeSpaceBone {boneName=handle}
    simSpaceSettings {maxAngularAcceleration=10, maxAngularVelocity=5, simSpaceAngularAlpha=0.2}
   }
   n31 = Anim Dynamics - Bone: ribbon_o_02!(ComponentPose=@n30, bAlphaBoolEnabled:bool=True, AngularSpringConstant:float=500, ExternalForce="0,0,0") -> (Pose) {
    alphaInputType=Bool
    angularSpringConstant=500
    bAngularSpring=true
    bUseGravityOverride=true
    simulationSpace=RootRelative
    alphaBoolBlend {blendOption=Linear}
    boundBone {boneName=ribbon_o_02}
    componentAppliedLinearAccClamp=<500 500 500>
    componentLinearAccScale=<0.2 0.2 0.2>
    componentLinearVelScale=<0.4 0.4 0.4>
    gravityOverride {z=-980}
    physicsBodyDefinitions [{
      boundBone {boneName=ribbon_o_02}
      boxExtents=<4 4 4>
      constraintSetup {angularLimitsMax=<15 40 35>, angularLimitsMin=<-10 -60 -30>, angularTarget=<0 0 0>}
      localJointOffset=<0.1 0 0>
     }]
    relativeSpaceBone {boneName=handle}
    simSpaceSettings {maxAngularAcceleration=10, maxAngularVelocity=5, simSpaceAngularAlpha=0.2}
   }
   n32 = Anim Dynamics - Bone: ribbon_p_01!(ComponentPose=@n31, bAlphaBoolEnabled:bool=True, AngularSpringConstant:float=500, ExternalForce="0,0,0") -> (Pose) {
    alphaInputType=Bool
    angularSpringConstant=500
    bAngularSpring=true
    bUseGravityOverride=true
    simulationSpace=RootRelative
    alphaBoolBlend {blendOption=Linear}
    boundBone {boneName=ribbon_p_01}
    componentAppliedLinearAccClamp=<500 500 500>
    componentLinearAccScale=<0.2 0.2 0.2>
    componentLinearVelScale=<0.4 0.4 0.4>
    gravityOverride {z=-980}
    physicsBodyDefinitions [{
      boundBone {boneName=ribbon_p_01}
      boxExtents=<4 4 4>
      constraintSetup {angularLimitsMax=<15 35 60>, angularLimitsMin=<-15 -70 -60>, angularTarget=<0 0 0>}
      localJointOffset=<0.05 0 0>
     }]
    relativeSpaceBone {boneName=handle}
    simSpaceSettings {maxAngularAcceleration=10, maxAngularVelocity=5, simSpaceAngularAlpha=0.2}
   }
   n33 = Anim Dynamics - Bone: ribbon_p_02!(ComponentPose=@n32, bAlphaBoolEnabled:bool=True, AngularSpringConstant:float=500, ExternalForce="0,0,0") -> (Pose) {
    alphaInputType=Bool
    angularSpringConstant=500
    bAngularSpring=true
    bUseGravityOverride=true
    simulationSpace=RootRelative
    alphaBoolBlend {blendOption=Linear}
    boundBone {boneName=ribbon_p_02}
    componentAppliedLinearAccClamp=<500 500 500>
    componentLinearAccScale=<0.2 0.2 0.2>
    componentLinearVelScale=<0.4 0.4 0.4>
    gravityOverride {z=-980}
    physicsBodyDefinitions [{
      boundBone {boneName=ribbon_p_02}
      boxExtents=<4 4 4>
      constraintSetup {angularLimitsMax=<15 35 35>, angularLimitsMin=<-10 -60 -30>, angularTarget=<0 0 0>}
      localJointOffset=<0.1 0 0>
     }]
    relativeSpaceBone {boneName=handle}
    simSpaceSettings {maxAngularAcceleration=10, maxAngularVelocity=5, simSpaceAngularAlpha=0.2}
   }
   n34 = Anim Dynamics - Bone: ribbon_q_01!(ComponentPose=@n33, bAlphaBoolEnabled:bool=True, AngularSpringConstant:float=500, ExternalForce="0,0,0") -> (Pose) {
    alphaInputType=Bool
    angularSpringConstant=500
    bAngularSpring=true
    bUseGravityOverride=true
    simulationSpace=RootRelative
    alphaBoolBlend {blendOption=Linear}
    boundBone {boneName=ribbon_q_01}
    componentAppliedLinearAccClamp=<500 500 500>
    componentLinearAccScale=<0.2 0.2 0.2>
    componentLinearVelScale=<0.4 0.4 0.4>
    gravityOverride {z=-980}
    physicsBodyDefinitions [{
      boundBone {boneName=ribbon_q_01}
      boxExtents=<4 4 4>
      constraintSetup {angularLimitsMax=<15 100 45>, angularLimitsMin=<-10 -90 -90>, angularTarget=<0 0 0>}
      localJointOffset=<0.05 0 0>
     }]
    relativeSpaceBone {boneName=handle}
    simSpaceSettings {maxAngularAcceleration=10, maxAngularVelocity=5, simSpaceAngularAlpha=0.2}
   }
   n35 = Anim Dynamics - Bone: ribbon_r_01!(ComponentPose=@n34, bAlphaBoolEnabled:bool=True, AngularSpringConstant:float=500, ExternalForce="0,0,0") -> (Pose) {
    alphaInputType=Bool
    angularSpringConstant=500
    bAngularSpring=true
    bUseGravityOverride=true
    simulationSpace=RootRelative
    alphaBoolBlend {blendOption=Linear}
    boundBone {boneName=ribbon_r_01}
    componentAppliedLinearAccClamp=<500 500 500>
    componentLinearAccScale=<0.2 0.2 0.2>
    componentLinearVelScale=<0.4 0.4 0.4>
    gravityOverride {z=-980}
    physicsBodyDefinitions [{
      boundBone {boneName=ribbon_r_01}
      boxExtents=<4 4 4>
      constraintSetup {angularLimitsMax=<15 120 45>, angularLimitsMin=<-10 -90 -90>, angularTarget=<0 0 0>}
      localJointOffset=<0.05 0 0>
     }]
    relativeSpaceBone {boneName=handle}
    simSpaceSettings {maxAngularAcceleration=10, maxAngularVelocity=5, simSpaceAngularAlpha=0.2}
   }
   n36 = Anim Dynamics - Bone: ribbon_t_02!(ComponentPose=@n35, bAlphaBoolEnabled:bool=True, AngularSpringConstant:float=500, ExternalForce="0,0,0") -> (Pose) {
    alphaInputType=Bool
    angularSpringConstant=500
    bAngularSpring=true
    bUseGravityOverride=true
    simulationSpace=RootRelative
    alphaBoolBlend {blendOption=Linear}
    boundBone {boneName=ribbon_t_02}
    componentAppliedLinearAccClamp=<500 500 500>
    componentLinearAccScale=<0.2 0.2 0.2>
    componentLinearVelScale=<0.4 0.4 0.4>
    gravityOverride {z=-980}
    physicsBodyDefinitions [{
      boundBone {boneName=ribbon_t_02}
      boxExtents=<4 4 4>
      constraintSetup {angularLimitsMax=<15 60 45>, angularLimitsMin=<-10 -45 -30>, angularTarget=<0 0 0>}
      localJointOffset=<0.1 0 0>
     }]
    relativeSpaceBone {boneName=handle}
    simSpaceSettings {maxAngularAcceleration=10, maxAngularVelocity=5, simSpaceAngularAlpha=0.2}
   }
   n37 = Anim Dynamics - Bone: ribbon_s_01!(ComponentPose=@n36, bAlphaBoolEnabled:bool=True, AngularSpringConstant:float=500, ExternalForce="0,0,0") -> (Pose) {
    alphaInputType=Bool
    angularSpringConstant=500
    bAngularSpring=true
    bUseGravityOverride=true
    simulationSpace=RootRelative
    alphaBoolBlend {blendOption=Linear}
    boundBone {boneName=ribbon_s_01}
    componentAppliedLinearAccClamp=<500 500 500>
    componentLinearAccScale=<0.2 0.2 0.2>
    componentLinearVelScale=<0.4 0.4 0.4>
    gravityOverride {z=-980}
    physicsBodyDefinitions [{
      boundBone {boneName=ribbon_s_01}
      boxExtents=<4 4 4>
      constraintSetup {angularLimitsMax=<15 100 45>, angularLimitsMin=<-10 -90 -90>, angularTarget=<0 0 0>}
      localJointOffset=<0.05 0 0>
     }]
    relativeSpaceBone {boneName=handle}
    simSpaceSettings {maxAngularAcceleration=10, maxAngularVelocity=5, simSpaceAngularAlpha=0.2}
   }
   n38 = Anim Dynamics - Bone: ribbon_t_01!(ComponentPose=@n37, bAlphaBoolEnabled:bool=True, AngularSpringConstant:float=500, ExternalForce="0,0,0") -> (Pose) {
    alphaInputType=Bool
    angularSpringConstant=500
    bAngularSpring=true
    bUseGravityOverride=true
    simulationSpace=RootRelative
    alphaBoolBlend {blendOption=Linear}
    boundBone {boneName=ribbon_t_01}
    componentAppliedLinearAccClamp=<500 500 500>
    componentLinearAccScale=<0.2 0.2 0.2>
    componentLinearVelScale=<0.4 0.4 0.4>
    gravityOverride {z=-980}
    physicsBodyDefinitions [{
      boundBone {boneName=ribbon_t_01}
      boxExtents=<4 4 4>
      constraintSetup {angularLimitsMax=<15 100 45>, angularLimitsMin=<-10 -90 -90>, angularTarget=<0 0 0>}
      localJointOffset=<0.05 0 0>
     }]
    relativeSpaceBone {boneName=handle}
    simSpaceSettings {maxAngularAcceleration=10, maxAngularVelocity=5, simSpaceAngularAlpha=0.2}
   }
   n39 = Anim Dynamics - Bone: ribbon_t_02(ComponentPose=@n38, bAlphaBoolEnabled=True, AngularSpringConstant=500, ExternalForce="0,0,0") {
    alphaInputType=Bool
    angularSpringConstant=500
    bAngularSpring=true
    bUseGravityOverride=true
    simulationSpace=RootRelative
    alphaBoolBlend {blendOption=Linear}
    boundBone {boneName=ribbon_t_02}
    componentAppliedLinearAccClamp=<500 500 500>
    componentLinearAccScale=<0.2 0.2 0.2>
    componentLinearVelScale=<0.4 0.4 0.4>
    gravityOverride {z=-980}
    physicsBodyDefinitions [{
      boundBone {boneName=ribbon_t_02}
      boxExtents=<4 4 4>
      constraintSetup {angularLimitsMax=<15 60 45>, angularLimitsMin=<-10 -45 -30>, angularTarget=<0 0 0>}
      localJointOffset=<0.1 0 0>
     }]
    relativeSpaceBone {boneName=handle}
    simSpaceSettings {maxAngularAcceleration=10, maxAngularVelocity=5, simSpaceAngularAlpha=0.2}
   }
   n40 = Anim Dynamics - Bone: ribbon_u_01!(ComponentPose=@n39, bAlphaBoolEnabled:bool=True, AngularSpringConstant:float=500, ExternalForce="0,0,0") -> (Pose) {
    alphaInputType=Bool
    angularSpringConstant=500
    bAngularSpring=true
    bUseGravityOverride=true
    simulationSpace=RootRelative
    alphaBoolBlend {blendOption=Linear}
    boundBone {boneName=ribbon_u_01}
    componentAppliedLinearAccClamp=<500 500 500>
    componentLinearAccScale=<0.2 0.2 0.2>
    componentLinearVelScale=<0.4 0.4 0.4>
    gravityOverride {z=-980}
    physicsBodyDefinitions [{
      boundBone {boneName=ribbon_u_01}
      boxExtents=<4 4 4>
      constraintSetup {angularLimitsMax=<15 100 45>, angularLimitsMin=<-10 -90 -90>, angularTarget=<0 0 0>}
      localJointOffset=<0.05 0 0>
     }]
    relativeSpaceBone {boneName=handle}
    simSpaceSettings {maxAngularAcceleration=10, maxAngularVelocity=5, simSpaceAngularAlpha=0.2}
   }
   n41 = Anim Dynamics - Bone: ribbon_u_02!(ComponentPose=@n40, bAlphaBoolEnabled:bool=True, AngularSpringConstant:float=500, ExternalForce="0,0,0") -> (Pose) {
    alphaInputType=Bool
    angularSpringConstant=500
    bAngularSpring=true
    bUseGravityOverride=true
    simulationSpace=RootRelative
    alphaBoolBlend {blendOption=Linear}
    boundBone {boneName=ribbon_u_02}
    componentAppliedLinearAccClamp=<500 500 500>
    componentLinearAccScale=<0.2 0.2 0.2>
    componentLinearVelScale=<0.4 0.4 0.4>
    gravityOverride {z=-980}
    physicsBodyDefinitions [{
      boundBone {boneName=ribbon_u_02}
      boxExtents=<4 4 4>
      constraintSetup {angularLimitsMax=<15 60 45>, angularLimitsMin=<-10 -45 -30>, angularTarget=<0 0 0>}
      localJointOffset=<0.1 0 0>
     }]
    relativeSpaceBone {boneName=handle}
    simSpaceSettings {maxAngularAcceleration=10, maxAngularVelocity=5, simSpaceAngularAlpha=0.2}
   }
   n42 = Anim Dynamics - Bone: ribbon_v_01!(ComponentPose=@n41, bAlphaBoolEnabled:bool=True, AngularSpringConstant:float=500, ExternalForce="0,0,0") -> (Pose) {
    alphaInputType=Bool
    angularSpringConstant=500
    bAngularSpring=true
    bUseGravityOverride=true
    simulationSpace=RootRelative
    alphaBoolBlend {blendOption=Linear}
    boundBone {boneName=ribbon_v_01}
    componentAppliedLinearAccClamp=<500 500 500>
    componentLinearAccScale=<0.2 0.2 0.2>
    componentLinearVelScale=<0.4 0.4 0.4>
    gravityOverride {z=-980}
    physicsBodyDefinitions [{
      boundBone {boneName=ribbon_v_01}
      boxExtents=<4 4 4>
      constraintSetup {angularLimitsMax=<15 100 45>, angularLimitsMin=<-10 -90 -90>, angularTarget=<0 0 0>}
      localJointOffset=<0.05 0 0>
     }]
    relativeSpaceBone {boneName=handle}
    simSpaceSettings {maxAngularAcceleration=10, maxAngularVelocity=5, simSpaceAngularAlpha=0.2}
   }
   n43 = Anim Dynamics - Bone: ribbon_v_02!(ComponentPose=@n42, bAlphaBoolEnabled:bool=True, AngularSpringConstant:float=500, ExternalForce="0,0,0") -> (Pose) {
    alphaInputType=Bool
    angularSpringConstant=500
    bAngularSpring=true
    bUseGravityOverride=true
    simulationSpace=RootRelative
    alphaBoolBlend {blendOption=Linear}
    boundBone {boneName=ribbon_v_02}
    componentAppliedLinearAccClamp=<500 500 500>
    componentLinearAccScale=<0.2 0.2 0.2>
    componentLinearVelScale=<0.4 0.4 0.4>
    gravityOverride {z=-980}
    physicsBodyDefinitions [{
      boundBone {boneName=ribbon_v_02}
      boxExtents=<4 4 4>
      constraintSetup {angularLimitsMax=<15 60 45>, angularLimitsMin=<-10 -45 -30>, angularTarget=<0 0 0>}
      localJointOffset=<0.1 0 0>
     }]
    relativeSpaceBone {boneName=handle}
    simSpaceSettings {maxAngularAcceleration=50, maxAngularVelocity=10, simSpaceAngularAlpha=0.5}
   }
   n44 = Anim Dynamics - Bone: ribbon_a_01!(ComponentPose=@n43, Alpha:float=1, AngularSpringConstant=500, ExternalForce="0,0,0") -> (Pose) {
    angularSpringConstant=500
    bAngularSpring=true
    bUseGravityOverride=true
    simulationSpace=RootRelative
    alphaBoolBlend {blendOption=Linear}
    boundBone {boneName=ribbon_a_01}
    componentAppliedLinearAccClamp=<1000 1000 1000>
    componentLinearAccScale=<0.5 0.5 0.5>
    componentLinearVelScale=<0.8 0.8 0.8>
    gravityOverride {z=-980}
    physicsBodyDefinitions [{
      boundBone {boneName=ribbon_a_01}
      boxExtents=<4 4 4>
      constraintSetup {angularLimitsMax=<15 60 20>, angularLimitsMin=<-10 -90 -20>, angularTarget=<0 0 0>}
      localJointOffset=<0.1 0 0>
     }]
    relativeSpaceBone {boneName=handle}
    simSpaceSettings {maxAngularAcceleration=50, maxAngularVelocity=1, simSpaceAngularAlpha=0.8}
   }
   n45 = Anim Dynamics - Bone: ribbon_b_01!(ComponentPose=@n44, Alpha:float=1, AngularSpringConstant=500, ExternalForce="0,0,0") -> (Pose) {
    angularSpringConstant=500
    bAngularSpring=true
    bUseGravityOverride=true
    simulationSpace=RootRelative
    alphaBoolBlend {blendOption=Linear}
    boundBone {boneName=ribbon_b_01}
    componentAppliedLinearAccClamp=<1000 1000 1000>
    componentLinearAccScale=<0.5 0.5 0.5>
    componentLinearVelScale=<0.8 0.8 0.8>
    gravityOverride {z=-980}
    physicsBodyDefinitions [{
      boundBone {boneName=ribbon_b_01}
      boxExtents=<4 4 4>
      constraintSetup {angularLimitsMax=<15 60 20>, angularLimitsMin=<-10 -90 -20>, angularTarget=<0 0 0>}
      localJointOffset=<0.1 0 0>
     }]
    relativeSpaceBone {boneName=handle}
    simSpaceSettings {maxAngularAcceleration=50, maxAngularVelocity=1, simSpaceAngularAlpha=0.8}
   }
   n46 = Anim Dynamics - Bone: ribbon_b_02!(ComponentPose=@n45, Alpha:float=1, AngularSpringConstant=500, ExternalForce="0,0,0") -> (Pose) {
    angularSpringConstant=500
    bAngularSpring=true
    bUseGravityOverride=true
    simulationSpace=RootRelative
    alphaBoolBlend {blendOption=Linear}
    boundBone {boneName=ribbon_b_02}
    componentAppliedLinearAccClamp=<1000 1000 1000>
    componentLinearAccScale=<0.5 0.5 0.5>
    componentLinearVelScale=<0.8 0.8 0.8>
    gravityOverride {z=-980}
    physicsBodyDefinitions [{
      boundBone {boneName=ribbon_b_02}
      boxExtents=<4 4 4>
      constraintSetup {angularLimitsMax=<15 20 25>, angularLimitsMin=<-15 -60 -30>, angularTarget=<0 0 0>}
      localJointOffset=<0.05 0 0>
     }]
    relativeSpaceBone {boneName=handle}
    simSpaceSettings {maxAngularAcceleration=50, maxAngularVelocity=1, simSpaceAngularAlpha=0.8}
   }
   n47 = Anim Dynamics - Bone: ribbon_c_01!(ComponentPose=@n46, Alpha:float=1, AngularSpringConstant=500, ExternalForce="0,0,0") -> (Pose) {
    angularSpringConstant=500
    bAngularSpring=true
    bUseGravityOverride=true
    simulationSpace=RootRelative
    alphaBoolBlend {blendOption=Linear}
    boundBone {boneName=ribbon_c_01}
    componentAppliedLinearAccClamp=<1000 1000 1000>
    componentLinearAccScale=<0.5 0.5 0.5>
    componentLinearVelScale=<0.8 0.8 0.8>
    gravityOverride {z=-980}
    physicsBodyDefinitions [{
      boundBone {boneName=ribbon_c_01}
      boxExtents=<4 4 4>
      constraintSetup {angularLimitsMax=<15 60 20>, angularLimitsMin=<-10 -90 -20>, angularTarget=<0 0 0>}
      localJointOffset=<0.1 0 0>
     }]
    relativeSpaceBone {boneName=handle}
    simSpaceSettings {maxAngularAcceleration=50, maxAngularVelocity=1, simSpaceAngularAlpha=0.8}
   }
   n48 = Anim Dynamics - Bone: ribbon_c_02!(ComponentPose=@n47, Alpha:float=1, AngularSpringConstant=500, ExternalForce="0,0,0") -> (Pose) {
    angularSpringConstant=500
    bAngularSpring=true
    bUseGravityOverride=true
    simulationSpace=RootRelative
    alphaBoolBlend {blendOption=Linear}
    boundBone {boneName=ribbon_c_02}
    componentAppliedLinearAccClamp=<1000 1000 1000>
    componentLinearAccScale=<0.5 0.5 0.5>
    componentLinearVelScale=<0.8 0.8 0.8>
    gravityOverride {z=-980}
    physicsBodyDefinitions [{
      boundBone {boneName=ribbon_c_02}
      boxExtents=<4 4 4>
      constraintSetup {angularLimitsMax=<15 20 25>, angularLimitsMin=<-15 -50 -30>, angularTarget=<0 0 0>}
      localJointOffset=<0.05 0 0>
     }]
    relativeSpaceBone {boneName=handle}
    simSpaceSettings {maxAngularAcceleration=50, maxAngularVelocity=1, simSpaceAngularAlpha=0.8}
   }
   n49 = Anim Dynamics - Bone: ribbon_c_03!(ComponentPose=@n48, Alpha:float=1, AngularSpringConstant=500, ExternalForce="0,0,0") -> (Pose) {
    angularSpringConstant=500
    bAngularSpring=true
    bUseGravityOverride=true
    simulationSpace=RootRelative
    alphaBoolBlend {blendOption=Linear}
    boundBone {boneName=ribbon_c_03}
    componentAppliedLinearAccClamp=<1000 1000 1000>
    componentLinearAccScale=<0.5 0.5 0.5>
    componentLinearVelScale=<0.8 0.8 0.8>
    gravityOverride {z=-980}
    physicsBodyDefinitions [{
      boundBone {boneName=ribbon_c_03}
      boxExtents=<4 4 4>
      constraintSetup {angularLimitsMax=<15 30 45>, angularLimitsMin=<-15 -45 -40>, angularTarget=<0 0 0>}
      localJointOffset=<0.05 0 0>
     }]
    relativeSpaceBone {boneName=handle}
    simSpaceSettings {maxAngularAcceleration=50, maxAngularVelocity=1, simSpaceAngularAlpha=0.8}
   }
   n50 = Anim Dynamics - Bone: ribbon_d_01!(ComponentPose=@n49, Alpha:float=1, AngularSpringConstant=500, ExternalForce="0,0,0") -> (Pose) {
    angularSpringConstant=500
    bAngularSpring=true
    bUseGravityOverride=true
    simulationSpace=RootRelative
    alphaBoolBlend {blendOption=Linear}
    boundBone {boneName=ribbon_d_01}
    componentAppliedLinearAccClamp=<1000 1000 1000>
    componentLinearAccScale=<0.5 0.5 0.5>
    componentLinearVelScale=<0.8 0.8 0.8>
    gravityOverride {z=-980}
    physicsBodyDefinitions [{
      boundBone {boneName=ribbon_d_01}
      boxExtents=<4 4 4>
      constraintSetup {angularLimitsMax=<15 60 25>, angularLimitsMin=<-10 -90 -30>, angularTarget=<0 0 0>}
      localJointOffset=<0.1 0 0>
     }]
    relativeSpaceBone {boneName=handle}
    simSpaceSettings {maxAngularAcceleration=50, maxAngularVelocity=1, simSpaceAngularAlpha=0.8}
   }
   n51 = Anim Dynamics - Bone: ribbon_d_02!(ComponentPose=@n50, Alpha:float=1, AngularSpringConstant=500, ExternalForce="0,0,0") -> (Pose) {
    angularSpringConstant=500
    bAngularSpring=true
    bUseGravityOverride=true
    simulationSpace=RootRelative
    alphaBoolBlend {blendOption=Linear}
    boundBone {boneName=ribbon_d_02}
    componentAppliedLinearAccClamp=<1000 1000 1000>
    componentLinearAccScale=<0.5 0.5 0.5>
    componentLinearVelScale=<0.8 0.8 0.8>
    gravityOverride {z=-980}
    physicsBodyDefinitions [{
      boundBone {boneName=ribbon_d_02}
      boxExtents=<4 4 4>
      constraintSetup {angularLimitsMax=<15 20 25>, angularLimitsMin=<-15 -50 -30>, angularTarget=<0 0 0>}
      localJointOffset=<0.05 0 0>
     }]
    relativeSpaceBone {boneName=handle}
    simSpaceSettings {maxAngularAcceleration=50, maxAngularVelocity=1, simSpaceAngularAlpha=0.8}
   }
   n52 = Anim Dynamics - Bone: ribbon_e_01!(ComponentPose=@n51, Alpha:float=1, AngularSpringConstant=500, ExternalForce="0,0,0") -> (Pose) {
    angularSpringConstant=500
    bAngularSpring=true
    bUseGravityOverride=true
    simulationSpace=RootRelative
    alphaBoolBlend {blendOption=Linear}
    boundBone {boneName=ribbon_e_01}
    componentAppliedLinearAccClamp=<1000 1000 1000>
    componentLinearAccScale=<0.5 0.5 0.5>
    componentLinearVelScale=<0.8 0.8 0.8>
    gravityOverride {z=-980}
    physicsBodyDefinitions [{
      boundBone {boneName=ribbon_e_01}
      boxExtents=<4 4 4>
      constraintSetup {angularLimitsMax=<15 60 20>, angularLimitsMin=<-10 -90 -10>, angularTarget=<0 0 0>}
      localJointOffset=<0.1 0 0>
     }]
    relativeSpaceBone {boneName=handle}
    simSpaceSettings {maxAngularAcceleration=50, maxAngularVelocity=1, simSpaceAngularAlpha=0.8}
   }
   n53 = Anim Dynamics - Bone: ribbon_f_01!(ComponentPose=@n52, Alpha:float=1, AngularSpringConstant=500, ExternalForce="0,0,0") -> (Pose) {
    angularSpringConstant=500
    bAngularSpring=true
    bUseGravityOverride=true
    simulationSpace=RootRelative
    alphaBoolBlend {blendOption=Linear}
    boundBone {boneName=ribbon_f_01}
    componentAppliedLinearAccClamp=<1000 1000 1000>
    componentLinearAccScale=<0.45 0.45 0.45>
    componentLinearVelScale=<0.85 0.85 0.85>
    gravityOverride {z=-980}
    physicsBodyDefinitions [{
      boundBone {boneName=ribbon_f_01}
      boxExtents=<4 4 4>
      constraintSetup {angularLimitsMax=<15 50 18>, angularLimitsMin=<-10 -70 -30>, angularTarget=<0 0 0>}
      localJointOffset=<0.1 0 0>
     }]
    relativeSpaceBone {boneName=handle}
    simSpaceSettings {maxAngularAcceleration=50, maxAngularVelocity=1, simSpaceAngularAlpha=0.8}
   }
   n54 = Anim Dynamics - Bone: ribbon_f_02!(ComponentPose=@n53, Alpha:float=1, AngularSpringConstant=500, ExternalForce="0,0,0") -> (Pose) {
    angularSpringConstant=500
    bAngularSpring=true
    bUseGravityOverride=true
    simulationSpace=RootRelative
    alphaBoolBlend {blendOption=Linear}
    boundBone {boneName=ribbon_f_02}
    componentAppliedLinearAccClamp=<1000 1000 1000>
    componentLinearAccScale=<0.45 0.45 0.45>
    componentLinearVelScale=<0.85 0.85 0.85>
    gravityOverride {z=-980}
    physicsBodyDefinitions [{
      boundBone {boneName=ribbon_f_02}
      boxExtents=<4 4 4>
      constraintSetup {angularLimitsMax=<15 30 20>, angularLimitsMin=<-10 -45 -12>, angularTarget=<0 0 0>}
      localJointOffset=<0.1 0 0>
     }]
    relativeSpaceBone {boneName=handle}
    simSpaceSettings {maxAngularAcceleration=50, maxAngularVelocity=1, simSpaceAngularAlpha=0.8}
   }
   n55 = Anim Dynamics - Bone: ribbon_g_01!(ComponentPose=@n54, Alpha:float=1, AngularSpringConstant=500, ExternalForce="0,0,0") -> (Pose) {
    angularSpringConstant=500
    bAngularSpring=true
    bUseGravityOverride=true
    simulationSpace=RootRelative
    alphaBoolBlend {blendOption=Linear}
    boundBone {boneName=ribbon_g_01}
    componentAppliedLinearAccClamp=<1000 1000 1000>
    componentLinearAccScale=<0.45 0.45 0.45>
    componentLinearVelScale=<0.85 0.85 0.85>
    gravityOverride {z=-980}
    physicsBodyDefinitions [{
      boundBone {boneName=ribbon_g_01}
      boxExtents=<4 4 4>
      constraintSetup {angularLimitsMax=<15 45 40>, angularLimitsMin=<-10 -60 -10>, angularTarget=<0 0 0>}
      localJointOffset=<0.1 0 0>
     }]
    relativeSpaceBone {boneName=handle}
    simSpaceSettings {maxAngularAcceleration=50, maxAngularVelocity=1, simSpaceAngularAlpha=0.8}
   }
   n56 = Anim Dynamics - Bone: ribbon_g_02!(ComponentPose=@n55, Alpha:float=1, AngularSpringConstant=500, ExternalForce="0,0,0") -> (Pose) {
    angularSpringConstant=500
    bAngularSpring=true
    bUseGravityOverride=true
    simulationSpace=RootRelative
    alphaBoolBlend {blendOption=Linear}
    boundBone {boneName=ribbon_g_02}
    componentAppliedLinearAccClamp=<1000 1000 1000>
    componentLinearAccScale=<0.45 0.45 0.45>
    componentLinearVelScale=<0.85 0.85 0.85>
    gravityOverride {z=-980}
    physicsBodyDefinitions [{
      boundBone {boneName=ribbon_g_02}
      boxExtents=<4 4 4>
      constraintSetup {angularLimitsMax=<15 10 20>, angularLimitsMin=<-10 -45 -60>, angularTarget=<0 0 0>}
      localJointOffset=<0.1 0 0>
     }]
    relativeSpaceBone {boneName=handle}
    simSpaceSettings {maxAngularAcceleration=50, maxAngularVelocity=1, simSpaceAngularAlpha=0.8}
   }
   n57 = Anim Dynamics - Bone: ribbon_h_01!(ComponentPose=@n56, Alpha:float=1, AngularSpringConstant=500, ExternalForce="0,0,0") -> (Pose) {
    angularSpringConstant=500
    bAngularSpring=true
    bUseGravityOverride=true
    simulationSpace=RootRelative
    alphaBoolBlend {blendOption=Linear}
    boundBone {boneName=ribbon_h_01}
    componentAppliedLinearAccClamp=<1000 1000 1000>
    componentLinearAccScale=<0.45 0.45 0.45>
    componentLinearVelScale=<0.85 0.85 0.85>
    gravityOverride {z=-980}
    physicsBodyDefinitions [{
      boundBone {boneName=ribbon_h_01}
      boxExtents=<4 4 4>
      constraintSetup {angularLimitsMax=<45 50 20>, angularLimitsMin=<-30 -45 -10>, angularTarget=<0 0 0>}
      localJointOffset=<0.1 0 0>
     }]
    relativeSpaceBone {boneName=handle}
    simSpaceSettings {maxAngularAcceleration=50, maxAngularVelocity=1, simSpaceAngularAlpha=0.8}
   }
   n58 = Anim Dynamics - Bone: ribbon_h_02!(ComponentPose=@n57, Alpha:float=1, AngularSpringConstant=500, ExternalForce="0,0,0") -> (Pose) {
    angularSpringConstant=500
    bAngularSpring=true
    bUseGravityOverride=true
    simulationSpace=RootRelative
    alphaBoolBlend {blendOption=Linear}
    boundBone {boneName=ribbon_h_02}
    componentAppliedLinearAccClamp=<1000 1000 1000>
    componentLinearAccScale=<0.45 0.45 0.45>
    componentLinearVelScale=<0.85 0.85 0.85>
    gravityOverride {z=-980}
    physicsBodyDefinitions [{
      boundBone {boneName=ribbon_h_02}
      boxExtents=<4 4 4>
      constraintSetup {angularLimitsMax=<15 60 25>, angularLimitsMin=<-10 -45 -30>, angularTarget=<0 0 0>}
      localJointOffset=<0.1 0 0>
     }]
    relativeSpaceBone {boneName=handle}
    simSpaceSettings {maxAngularAcceleration=50, maxAngularVelocity=1, simSpaceAngularAlpha=0.8}
   }
   n59 = Anim Dynamics - Bone: ribbon_h_03!(ComponentPose=@n58, Alpha:float=1, AngularSpringConstant=500, ExternalForce="0,0,0") -> (Pose) {
    angularSpringConstant=500
    bAngularSpring=true
    bUseGravityOverride=true
    simulationSpace=RootRelative
    alphaBoolBlend {blendOption=Linear}
    boundBone {boneName=ribbon_h_03}
    componentAppliedLinearAccClamp=<1000 1000 1000>
    componentLinearAccScale=<0.45 0.45 0.45>
    componentLinearVelScale=<0.85 0.85 0.85>
    gravityOverride {z=-980}
    physicsBodyDefinitions [{
      boundBone {boneName=ribbon_h_03}
      boxExtents=<4 4 4>
      constraintSetup {angularLimitsMax=<15 30 60>, angularLimitsMin=<-15 -45 -45>, angularTarget=<0 0 0>}
      localJointOffset=<0.1 0 0>
     }]
    relativeSpaceBone {boneName=handle}
    simSpaceSettings {maxAngularAcceleration=50, maxAngularVelocity=1, simSpaceAngularAlpha=0.8}
   }
   n60 = Anim Dynamics - Bone: ribbon_i_01!(ComponentPose=@n59, Alpha:float=1, AngularSpringConstant=500, ExternalForce="0,0,0") -> (Pose) {
    angularSpringConstant=500
    bAngularSpring=true
    bUseGravityOverride=true
    simulationSpace=RootRelative
    alphaBoolBlend {blendOption=Linear}
    boundBone {boneName=ribbon_i_01}
    componentAppliedLinearAccClamp=<1000 1000 1000>
    componentLinearAccScale=<0.45 0.45 0.45>
    componentLinearVelScale=<0.85 0.85 0.85>
    gravityOverride {z=-980}
    physicsBodyDefinitions [{
      boundBone {boneName=ribbon_i_01}
      boxExtents=<4 4 4>
      constraintSetup {angularLimitsMax=<15 10 15>, angularLimitsMin=<-10 -70 -20>, angularTarget=<0 0 0>}
      localJointOffset=<0.1 0 0>
     }]
    relativeSpaceBone {boneName=handle}
    simSpaceSettings {maxAngularAcceleration=50, maxAngularVelocity=1, simSpaceAngularAlpha=0.8}
   }
   n61 = Anim Dynamics - Bone: ribbon_i_02!(ComponentPose=@n60, Alpha:float=1, AngularSpringConstant=500, ExternalForce="0,0,0") -> (Pose) {
    angularSpringConstant=500
    bAngularSpring=true
    bUseGravityOverride=true
    simulationSpace=RootRelative
    alphaBoolBlend {blendOption=Linear}
    boundBone {boneName=ribbon_i_02}
    componentAppliedLinearAccClamp=<1000 1000 1000>
    componentLinearAccScale=<0.45 0.45 0.45>
    componentLinearVelScale=<0.85 0.85 0.85>
    gravityOverride {z=-980}
    physicsBodyDefinitions [{
      boundBone {boneName=ribbon_i_02}
      boxExtents=<5 4 4>
      constraintSetup {angularLimitsMax=<15 10 20>, angularLimitsMin=<-10 -45 -30>, angularTarget=<0 0 0>}
      localJointOffset=<0.1 0 0>
     }]
    relativeSpaceBone {boneName=handle}
    simSpaceSettings {maxAngularAcceleration=50, maxAngularVelocity=1, simSpaceAngularAlpha=0.8}
   }
   n62 = Anim Dynamics - Bone: ribbon_j_01!(ComponentPose=@n61, Alpha:float=1, AngularSpringConstant=500, ExternalForce="0,0,0") -> (Pose) {
    angularSpringConstant=500
    bAngularSpring=true
    bUseGravityOverride=true
    simulationSpace=RootRelative
    alphaBoolBlend {blendOption=Linear}
    boundBone {boneName=ribbon_j_01}
    componentAppliedLinearAccClamp=<1000 1000 1000>
    componentLinearAccScale=<0.45 0.45 0.45>
    componentLinearVelScale=<0.85 0.85 0.85>
    gravityOverride {z=-980}
    physicsBodyDefinitions [{
      boundBone {boneName=ribbon_j_01}
      boxExtents=<4 4 4>
      constraintSetup {angularLimitsMax=<15 10 10>, angularLimitsMin=<-10 -70 -20>, angularTarget=<0 0 0>}
      localJointOffset=<0.1 0 0>
     }]
    relativeSpaceBone {boneName=handle}
    simSpaceSettings {maxAngularAcceleration=50, maxAngularVelocity=1, simSpaceAngularAlpha=0.8}
   }
   n63 = Anim Dynamics - Bone: ribbon_j_02!(ComponentPose=@n62, Alpha:float=1, AngularSpringConstant=500, ExternalForce="0,0,0") -> (Pose) {
    angularSpringConstant=500
    bAngularSpring=true
    bUseGravityOverride=true
    simulationSpace=RootRelative
    alphaBoolBlend {blendOption=Linear}
    boundBone {boneName=ribbon_j_02}
    componentAppliedLinearAccClamp=<1000 1000 1000>
    componentLinearAccScale=<0.45 0.45 0.45>
    componentLinearVelScale=<0.85 0.85 0.85>
    gravityOverride {z=-980}
    physicsBodyDefinitions [{
      boundBone {boneName=ribbon_j_02}
      boxExtents=<5 4 4>
      constraintSetup {angularLimitsMax=<15 60 20>, angularLimitsMin=<-10 -45 -30>, angularTarget=<0 0 0>}
      localJointOffset=<0.1 0 0>
     }]
    relativeSpaceBone {boneName=handle}
    simSpaceSettings {maxAngularAcceleration=50, maxAngularVelocity=1, simSpaceAngularAlpha=0.8}
   }
   n64 = Anim Dynamics - Bone: ribbon_k_01!(ComponentPose=@n63, Alpha:float=1, AngularSpringConstant=500, ExternalForce="0,0,0") -> (Pose) {
    angularSpringConstant=500
    bAngularSpring=true
    bUseGravityOverride=true
    simulationSpace=RootRelative
    alphaBoolBlend {blendOption=Linear}
    boundBone {boneName=ribbon_k_01}
    componentAppliedLinearAccClamp=<1000 1000 1000>
    componentLinearAccScale=<0.4 0.4 0.4>
    componentLinearVelScale=<0.8 0.8 0.8>
    gravityOverride {z=-980}
    physicsBodyDefinitions [{
      boundBone {boneName=ribbon_k_01}
      boxExtents=<4 4 4>
      constraintSetup {angularLimitsMax=<15 28 55>, angularLimitsMin=<-15 -60 -60>, angularTarget=<0 0 0>}
      localJointOffset=<0.05 0 0>
     }]
    relativeSpaceBone {boneName=handle}
    simSpaceSettings {maxAngularAcceleration=50, maxAngularVelocity=1, simSpaceAngularAlpha=0.8}
   }
   n65 = Anim Dynamics - Bone: ribbon_k_02!(ComponentPose=@n64, Alpha:float=1, AngularSpringConstant=500, ExternalForce="0,0,0") -> (Pose) {
    angularSpringConstant=500
    bAngularSpring=true
    bUseGravityOverride=true
    simulationSpace=RootRelative
    alphaBoolBlend {blendOption=Linear}
    boundBone {boneName=ribbon_k_02}
    componentAppliedLinearAccClamp=<1000 1000 1000>
    componentLinearAccScale=<0.4 0.4 0.4>
    componentLinearVelScale=<0.8 0.8 0.8>
    gravityOverride {z=-980}
    physicsBodyDefinitions [{
      boundBone {boneName=ribbon_k_02}
      boxExtents=<4 4 4>
      constraintSetup {angularLimitsMax=<15 10 35>, angularLimitsMin=<-10 -60 -30>, angularTarget=<0 0 0>}
      localJointOffset=<0.1 0 0>
     }]
    relativeSpaceBone {boneName=handle}
    simSpaceSettings {maxAngularAcceleration=50, maxAngularVelocity=1, simSpaceAngularAlpha=0.8}
   }
   n66 = Anim Dynamics - Bone: ribbon_k_03!(ComponentPose=@n65, Alpha:float=1, AngularSpringConstant=500, ExternalForce="0,0,0") -> (Pose) {
    angularSpringConstant=500
    bAngularSpring=true
    bUseGravityOverride=true
    simulationSpace=RootRelative
    alphaBoolBlend {blendOption=Linear}
    boundBone {boneName=ribbon_k_03}
    componentAppliedLinearAccClamp=<1000 1000 1000>
    componentLinearAccScale=<0.4 0.4 0.4>
    componentLinearVelScale=<0.8 0.8 0.8>
    gravityOverride {z=-980}
    physicsBodyDefinitions [{
      boundBone {boneName=ribbon_k_03}
      boxExtents=<4 4 4>
      constraintSetup {angularLimitsMax=<15 15 45>, angularLimitsMin=<-10 -30 -45>, angularTarget=<0 0 0>}
      localJointOffset=<0.1 0 0>
     }]
    relativeSpaceBone {boneName=handle}
    simSpaceSettings {maxAngularAcceleration=50, maxAngularVelocity=1, simSpaceAngularAlpha=0.8}
   }
   n67 = Anim Dynamics - Bone: ribbon_l_01!(ComponentPose=@n66, Alpha:float=1, AngularSpringConstant=500, ExternalForce="0,0,0") -> (Pose) {
    angularSpringConstant=500
    bAngularSpring=true
    bUseGravityOverride=true
    simulationSpace=RootRelative
    alphaBoolBlend {blendOption=Linear}
    boundBone {boneName=ribbon_l_01}
    componentAppliedLinearAccClamp=<1000 1000 1000>
    componentLinearAccScale=<0.4 0.4 0.4>
    componentLinearVelScale=<0.8 0.8 0.8>
    gravityOverride {z=-980}
    physicsBodyDefinitions [{
      boundBone {boneName=ribbon_l_01}
      boxExtents=<4 4 4>
      constraintSetup {angularLimitsMax=<15 20 60>, angularLimitsMin=<-15 -60 -60>, angularTarget=<0 0 0>}
      localJointOffset=<0.05 0 0>
     }]
    relativeSpaceBone {boneName=handle}
    simSpaceSettings {maxAngularAcceleration=50, maxAngularVelocity=1, simSpaceAngularAlpha=0.8}
   }
   n68 = Anim Dynamics - Bone: ribbon_l_02!(ComponentPose=@n67, Alpha:float=1, AngularSpringConstant=500, ExternalForce="0,0,0") -> (Pose) {
    angularSpringConstant=500
    bAngularSpring=true
    bUseGravityOverride=true
    simulationSpace=RootRelative
    alphaBoolBlend {blendOption=Linear}
    boundBone {boneName=ribbon_l_02}
    componentAppliedLinearAccClamp=<1000 1000 1000>
    componentLinearAccScale=<0.4 0.4 0.4>
    componentLinearVelScale=<0.8 0.8 0.8>
    gravityOverride {z=-980}
    physicsBodyDefinitions [{
      boundBone {boneName=ribbon_l_02}
      boxExtents=<4 4 4>
      constraintSetup {angularLimitsMax=<15 35 35>, angularLimitsMin=<-10 -60 -30>, angularTarget=<0 0 0>}
      localJointOffset=<0.1 0 0>
     }]
    relativeSpaceBone {boneName=handle}
    simSpaceSettings {maxAngularAcceleration=50, maxAngularVelocity=1, simSpaceAngularAlpha=0.8}
   }
   n69 = Anim Dynamics - Bone: ribbon_m_01!(ComponentPose=@n68, Alpha:float=1, AngularSpringConstant=500, ExternalForce="0,0,0") -> (Pose) {
    angularSpringConstant=500
    bAngularSpring=true
    bUseGravityOverride=true
    simulationSpace=RootRelative
    alphaBoolBlend {blendOption=Linear}
    boundBone {boneName=ribbon_m_01}
    componentAppliedLinearAccClamp=<1000 1000 1000>
    componentLinearAccScale=<0.4 0.4 0.4>
    componentLinearVelScale=<0.8 0.8 0.8>
    gravityOverride {z=-980}
    physicsBodyDefinitions [{
      boundBone {boneName=ribbon_m_01}
      boxExtents=<4 4 4>
      constraintSetup {angularLimitsMax=<15 30 55>, angularLimitsMin=<-15 -60 -30>, angularTarget=<0 0 0>}
      localJointOffset=<0.05 0 0>
     }]
    relativeSpaceBone {boneName=handle}
    simSpaceSettings {maxAngularAcceleration=50, maxAngularVelocity=1, simSpaceAngularAlpha=0.8}
   }
   n70 = Anim Dynamics - Bone: ribbon_m_02!(ComponentPose=@n69, Alpha:float=1, AngularSpringConstant=500, ExternalForce="0,0,0") -> (Pose) {
    angularSpringConstant=500
    bAngularSpring=true
    bUseGravityOverride=true
    simulationSpace=RootRelative
    alphaBoolBlend {blendOption=Linear}
    boundBone {boneName=ribbon_m_02}
    componentAppliedLinearAccClamp=<1000 1000 1000>
    componentLinearAccScale=<0.4 0.4 0.4>
    componentLinearVelScale=<0.8 0.8 0.8>
    gravityOverride {z=-980}
    physicsBodyDefinitions [{
      boundBone {boneName=ribbon_m_02}
      boxExtents=<4 4 4>
      constraintSetup {angularLimitsMax=<15 35 35>, angularLimitsMin=<-10 -60 -30>, angularTarget=<0 0 0>}
      localJointOffset=<0.1 0 0>
     }]
    relativeSpaceBone {boneName=handle}
    simSpaceSettings {maxAngularAcceleration=50, maxAngularVelocity=1, simSpaceAngularAlpha=0.8}
   }
   n71 = Anim Dynamics - Bone: ribbon_m_03!(ComponentPose=@n70, Alpha:float=1, AngularSpringConstant=500, ExternalForce="0,0,0") -> (Pose) {
    angularSpringConstant=500
    bAngularSpring=true
    bUseGravityOverride=true
    simulationSpace=RootRelative
    alphaBoolBlend {blendOption=Linear}
    boundBone {boneName=ribbon_m_03}
    componentAppliedLinearAccClamp=<1000 1000 1000>
    componentLinearAccScale=<0.4 0.4 0.4>
    componentLinearVelScale=<0.8 0.8 0.8>
    gravityOverride {z=-980}
    physicsBodyDefinitions [{
      boundBone {boneName=ribbon_m_03}
      boxExtents=<4 4 4>
      constraintSetup {angularLimitsMax=<15 45 45>, angularLimitsMin=<-10 -30 -45>, angularTarget=<0 0 0>}
      localJointOffset=<0.1 0 0>
     }]
    relativeSpaceBone {boneName=handle}
    simSpaceSettings {maxAngularAcceleration=50, maxAngularVelocity=1, simSpaceAngularAlpha=0.8}
   }
   n72 = Anim Dynamics - Bone: ribbon_n_01!(ComponentPose=@n71, Alpha:float=1, AngularSpringConstant=500, ExternalForce="0,0,0") -> (Pose) {
    angularSpringConstant=500
    bAngularSpring=true
    bUseGravityOverride=true
    simulationSpace=RootRelative
    alphaBoolBlend {blendOption=Linear}
    boundBone {boneName=ribbon_n_01}
    componentAppliedLinearAccClamp=<1000 1000 1000>
    componentLinearAccScale=<0.4 0.4 0.4>
    componentLinearVelScale=<0.8 0.8 0.8>
    gravityOverride {z=-980}
    physicsBodyDefinitions [{
      boundBone {boneName=ribbon_n_01}
      boxExtents=<4 4 4>
      constraintSetup {angularLimitsMax=<15 30 25>, angularLimitsMin=<-15 -60 -20>, angularTarget=<0 0 0>}
      localJointOffset=<0.1 0 0>
     }]
    relativeSpaceBone {boneName=handle}
    simSpaceSettings {maxAngularAcceleration=50, maxAngularVelocity=1, simSpaceAngularAlpha=0.8}
   }
   n73 = Anim Dynamics - Bone: ribbon_o_01!(ComponentPose=@n72, Alpha:float=1, AngularSpringConstant=500, ExternalForce="0,0,0") -> (Pose) {
    angularSpringConstant=500
    bAngularSpring=true
    bUseGravityOverride=true
    simulationSpace=RootRelative
    alphaBoolBlend {blendOption=Linear}
    boundBone {boneName=ribbon_o_01}
    componentAppliedLinearAccClamp=<1000 1000 1000>
    componentLinearAccScale=<0.4 0.4 0.4>
    componentLinearVelScale=<0.8 0.8 0.8>
    gravityOverride {z=-980}
    physicsBodyDefinitions [{
      boundBone {boneName=ribbon_o_01}
      boxExtents=<4 4 4>
      constraintSetup {angularLimitsMax=<0 35 30>, angularLimitsMin=<0 -60 -45>, angularTarget=<0 0 0>}
      localJointOffset=<0.05 0 0>
     }]
    relativeSpaceBone {boneName=handle}
    simSpaceSettings {maxAngularAcceleration=50, maxAngularVelocity=1, simSpaceAngularAlpha=0.8}
   }
   n74 = Anim Dynamics - Bone: ribbon_o_02!(ComponentPose=@n73, Alpha:float=1, AngularSpringConstant=500, ExternalForce="0,0,0") -> (Pose) {
    angularSpringConstant=500
    bAngularSpring=true
    bUseGravityOverride=true
    simulationSpace=RootRelative
    alphaBoolBlend {blendOption=Linear}
    boundBone {boneName=ribbon_o_02}
    componentAppliedLinearAccClamp=<1000 1000 1000>
    componentLinearAccScale=<0.4 0.4 0.4>
    componentLinearVelScale=<0.8 0.8 0.8>
    gravityOverride {z=-980}
    physicsBodyDefinitions [{
      boundBone {boneName=ribbon_o_02}
      boxExtents=<4 4 4>
      constraintSetup {angularLimitsMax=<15 40 35>, angularLimitsMin=<-10 -60 -30>, angularTarget=<0 0 0>}
      localJointOffset=<0.1 0 0>
     }]
    relativeSpaceBone {boneName=handle}
    simSpaceSettings {maxAngularAcceleration=50, maxAngularVelocity=1, simSpaceAngularAlpha=0.8}
   }
   n75 = Anim Dynamics - Bone: ribbon_p_01!(ComponentPose=@n74, Alpha:float=1, AngularSpringConstant=500, ExternalForce="0,0,0") -> (Pose) {
    angularSpringConstant=500
    bAngularSpring=true
    bUseGravityOverride=true
    simulationSpace=RootRelative
    alphaBoolBlend {blendOption=Linear}
    boundBone {boneName=ribbon_p_01}
    componentAppliedLinearAccClamp=<1000 1000 1000>
    componentLinearAccScale=<0.4 0.4 0.4>
    componentLinearVelScale=<0.8 0.8 0.8>
    gravityOverride {z=-980}
    physicsBodyDefinitions [{
      boundBone {boneName=ribbon_p_01}
      boxExtents=<4 4 4>
      constraintSetup {angularLimitsMax=<15 35 60>, angularLimitsMin=<-15 -70 -60>, angularTarget=<0 0 0>}
      localJointOffset=<0.05 0 0>
     }]
    relativeSpaceBone {boneName=handle}
    simSpaceSettings {maxAngularAcceleration=50, maxAngularVelocity=1, simSpaceAngularAlpha=0.8}
   }
   n76 = Anim Dynamics - Bone: ribbon_p_02!(ComponentPose=@n75, Alpha:float=1, AngularSpringConstant=500, ExternalForce="0,0,0") -> (Pose) {
    angularSpringConstant=500
    bAngularSpring=true
    bUseGravityOverride=true
    simulationSpace=RootRelative
    alphaBoolBlend {blendOption=Linear}
    boundBone {boneName=ribbon_p_02}
    componentAppliedLinearAccClamp=<1000 1000 1000>
    componentLinearAccScale=<0.4 0.4 0.4>
    componentLinearVelScale=<0.8 0.8 0.8>
    gravityOverride {z=-980}
    physicsBodyDefinitions [{
      boundBone {boneName=ribbon_p_02}
      boxExtents=<4 4 4>
      constraintSetup {angularLimitsMax=<15 35 35>, angularLimitsMin=<-10 -60 -30>, angularTarget=<0 0 0>}
      localJointOffset=<0.1 0 0>
     }]
    relativeSpaceBone {boneName=handle}
    simSpaceSettings {maxAngularAcceleration=50, maxAngularVelocity=1, simSpaceAngularAlpha=0.8}
   }
   n77 = Anim Dynamics - Bone: ribbon_q_01!(ComponentPose=@n76, Alpha:float=1, AngularSpringConstant=500, ExternalForce="0,0,0") -> (Pose) {
    angularSpringConstant=500
    bAngularSpring=true
    bUseGravityOverride=true
    simulationSpace=RootRelative
    alphaBoolBlend {blendOption=Linear}
    boundBone {boneName=ribbon_q_01}
    componentAppliedLinearAccClamp=<1000 1000 1000>
    componentLinearAccScale=<0.35 0.35 0.35>
    componentLinearVelScale=<0.8 0.8 0.8>
    gravityOverride {z=-980}
    physicsBodyDefinitions [{
      boundBone {boneName=ribbon_q_01}
      boxExtents=<4 4 4>
      constraintSetup {angularLimitsMax=<15 100 45>, angularLimitsMin=<-10 -90 -90>, angularTarget=<0 0 0>}
      localJointOffset=<0.05 0 0>
     }]
    relativeSpaceBone {boneName=handle}
    simSpaceSettings {maxAngularAcceleration=50, maxAngularVelocity=1, simSpaceAngularAlpha=0.8}
   }
   n78 = Anim Dynamics - Bone: ribbon_r_01!(ComponentPose=@n77, Alpha:float=1, AngularSpringConstant=500, ExternalForce="0,0,0") -> (Pose) {
    angularSpringConstant=500
    bAngularSpring=true
    bUseGravityOverride=true
    simulationSpace=RootRelative
    alphaBoolBlend {blendOption=Linear}
    boundBone {boneName=ribbon_r_01}
    componentAppliedLinearAccClamp=<1000 1000 1000>
    componentLinearAccScale=<0.35 0.35 0.35>
    componentLinearVelScale=<0.8 0.8 0.8>
    gravityOverride {z=-980}
    physicsBodyDefinitions [{
      boundBone {boneName=ribbon_r_01}
      boxExtents=<4 4 4>
      constraintSetup {angularLimitsMax=<15 120 45>, angularLimitsMin=<-10 -90 -90>, angularTarget=<0 0 0>}
      localJointOffset=<0.05 0 0>
     }]
    relativeSpaceBone {boneName=handle}
    simSpaceSettings {maxAngularAcceleration=50, maxAngularVelocity=1, simSpaceAngularAlpha=0.8}
   }
   n79 = Anim Dynamics - Bone: ribbon_t_02!(ComponentPose=@n78, Alpha:float=1, AngularSpringConstant=500, ExternalForce="0,0,0") -> (Pose) {
    angularSpringConstant=500
    bAngularSpring=true
    bUseGravityOverride=true
    simulationSpace=RootRelative
    alphaBoolBlend {blendOption=Linear}
    boundBone {boneName=ribbon_t_02}
    componentAppliedLinearAccClamp=<1000 1000 1000>
    componentLinearAccScale=<0.35 0.35 0.35>
    componentLinearVelScale=<0.8 0.8 0.8>
    gravityOverride {z=-980}
    physicsBodyDefinitions [{
      boundBone {boneName=ribbon_t_02}
      boxExtents=<4 4 4>
      constraintSetup {angularLimitsMax=<15 60 45>, angularLimitsMin=<-10 -45 -30>, angularTarget=<0 0 0>}
      localJointOffset=<0.1 0 0>
     }]
    relativeSpaceBone {boneName=handle}
    simSpaceSettings {maxAngularAcceleration=50, maxAngularVelocity=1, simSpaceAngularAlpha=0.8}
   }
   n80 = Anim Dynamics - Bone: ribbon_s_01!(ComponentPose=@n79, Alpha:float=1, AngularSpringConstant=500, ExternalForce="0,0,0") -> (Pose) {
    angularSpringConstant=500
    bAngularSpring=true
    bUseGravityOverride=true
    simulationSpace=RootRelative
    alphaBoolBlend {blendOption=Linear}
    boundBone {boneName=ribbon_s_01}
    componentAppliedLinearAccClamp=<1000 1000 1000>
    componentLinearAccScale=<0.35 0.35 0.35>
    componentLinearVelScale=<0.8 0.8 0.8>
    gravityOverride {z=-980}
    physicsBodyDefinitions [{
      boundBone {boneName=ribbon_s_01}
      boxExtents=<4 4 4>
      constraintSetup {angularLimitsMax=<15 100 45>, angularLimitsMin=<-10 -90 -90>, angularTarget=<0 0 0>}
      localJointOffset=<0.05 0 0>
     }]
    relativeSpaceBone {boneName=handle}
    simSpaceSettings {maxAngularAcceleration=50, maxAngularVelocity=1, simSpaceAngularAlpha=0.8}
   }
   n81 = Anim Dynamics - Bone: ribbon_t_01!(ComponentPose=@n80, Alpha:float=1, AngularSpringConstant=500, ExternalForce="0,0,0") -> (Pose) {
    angularSpringConstant=500
    bAngularSpring=true
    bUseGravityOverride=true
    simulationSpace=RootRelative
    alphaBoolBlend {blendOption=Linear}
    boundBone {boneName=ribbon_t_01}
    componentAppliedLinearAccClamp=<1000 1000 1000>
    componentLinearAccScale=<0.35 0.35 0.35>
    componentLinearVelScale=<0.8 0.8 0.8>
    gravityOverride {z=-980}
    physicsBodyDefinitions [{
      boundBone {boneName=ribbon_t_01}
      boxExtents=<4 4 4>
      constraintSetup {angularLimitsMax=<15 100 45>, angularLimitsMin=<-10 -90 -90>, angularTarget=<0 0 0>}
      localJointOffset=<0.05 0 0>
     }]
    relativeSpaceBone {boneName=handle}
    simSpaceSettings {maxAngularAcceleration=50, maxAngularVelocity=1, simSpaceAngularAlpha=0.8}
   }
   n82 = Anim Dynamics - Bone: ribbon_t_02(ComponentPose=@n81, Alpha=1, AngularSpringConstant=500, ExternalForce="0,0,0") {
    angularSpringConstant=500
    bAngularSpring=true
    bUseGravityOverride=true
    simulationSpace=RootRelative
    alphaBoolBlend {blendOption=Linear}
    boundBone {boneName=ribbon_t_02}
    componentAppliedLinearAccClamp=<1000 1000 1000>
    componentLinearAccScale=<0.35 0.35 0.35>
    componentLinearVelScale=<0.8 0.8 0.8>
    gravityOverride {z=-980}
    physicsBodyDefinitions [{
      boundBone {boneName=ribbon_t_02}
      boxExtents=<4 4 4>
      constraintSetup {angularLimitsMax=<15 60 45>, angularLimitsMin=<-10 -45 -30>, angularTarget=<0 0 0>}
      localJointOffset=<0.1 0 0>
     }]
    relativeSpaceBone {boneName=handle}
    simSpaceSettings {maxAngularAcceleration=50, maxAngularVelocity=1, simSpaceAngularAlpha=0.8}
   }
   n83 = Anim Dynamics - Bone: ribbon_u_01!(ComponentPose=@n82, Alpha:float=1, AngularSpringConstant=500, ExternalForce="0,0,0") -> (Pose) {
    angularSpringConstant=500
    bAngularSpring=true
    bUseGravityOverride=true
    simulationSpace=RootRelative
    alphaBoolBlend {blendOption=Linear}
    boundBone {boneName=ribbon_u_01}
    componentAppliedLinearAccClamp=<1000 1000 1000>
    componentLinearAccScale=<0.35 0.35 0.35>
    componentLinearVelScale=<0.8 0.8 0.8>
    gravityOverride {z=-980}
    physicsBodyDefinitions [{
      boundBone {boneName=ribbon_u_01}
      boxExtents=<4 4 4>
      constraintSetup {angularLimitsMax=<15 100 45>, angularLimitsMin=<-10 -90 -90>, angularTarget=<0 0 0>}
      localJointOffset=<0.05 0 0>
     }]
    relativeSpaceBone {boneName=handle}
    simSpaceSettings {maxAngularAcceleration=50, maxAngularVelocity=1, simSpaceAngularAlpha=0.8}
   }
   n84 = Anim Dynamics - Bone: ribbon_u_02!(ComponentPose=@n83, Alpha:float=1, AngularSpringConstant=500, ExternalForce="0,0,0") -> (Pose) {
    angularSpringConstant=500
    bAngularSpring=true
    bUseGravityOverride=true
    simulationSpace=RootRelative
    alphaBoolBlend {blendOption=Linear}
    boundBone {boneName=ribbon_u_02}
    componentAppliedLinearAccClamp=<1000 1000 1000>
    componentLinearAccScale=<0.35 0.35 0.35>
    componentLinearVelScale=<0.8 0.8 0.8>
    gravityOverride {z=-980}
    physicsBodyDefinitions [{
      boundBone {boneName=ribbon_u_02}
      boxExtents=<4 4 4>
      constraintSetup {angularLimitsMax=<15 60 45>, angularLimitsMin=<-10 -45 -30>, angularTarget=<0 0 0>}
      localJointOffset=<0.1 0 0>
     }]
    relativeSpaceBone {boneName=handle}
    simSpaceSettings {maxAngularAcceleration=50, maxAngularVelocity=1, simSpaceAngularAlpha=0.8}
   }
   n85 = Anim Dynamics - Bone: ribbon_v_01!(ComponentPose=@n84, Alpha:float=1, AngularSpringConstant=500, ExternalForce="0,0,0") -> (Pose) {
    angularSpringConstant=500
    bAngularSpring=true
    bUseGravityOverride=true
    simulationSpace=RootRelative
    alphaBoolBlend {blendOption=Linear}
    boundBone {boneName=ribbon_v_01}
    componentAppliedLinearAccClamp=<1000 1000 1000>
    componentLinearAccScale=<0.35 0.35 0.35>
    componentLinearVelScale=<0.8 0.8 0.8>
    gravityOverride {z=-980}
    physicsBodyDefinitions [{
      boundBone {boneName=ribbon_v_01}
      boxExtents=<4 4 4>
      constraintSetup {angularLimitsMax=<15 100 45>, angularLimitsMin=<-10 -90 -90>, angularTarget=<0 0 0>}
      localJointOffset=<0.05 0 0>
     }]
    relativeSpaceBone {boneName=handle}
    simSpaceSettings {maxAngularAcceleration=50, maxAngularVelocity=1, simSpaceAngularAlpha=0.8}
   }
   n86 = Anim Dynamics - Bone: ribbon_v_02!(ComponentPose=@n85, Alpha:float=1, AngularSpringConstant=500, ExternalForce="0,0,0") -> (Pose) {
    angularSpringConstant=500
    bAngularSpring=true
    bUseGravityOverride=true
    simulationSpace=RootRelative
    alphaBoolBlend {blendOption=Linear}
    boundBone {boneName=ribbon_v_02}
    componentAppliedLinearAccClamp=<1000 1000 1000>
    componentLinearAccScale=<0.35 0.35 0.35>
    componentLinearVelScale=<0.8 0.8 0.8>
    gravityOverride {z=-980}
    physicsBodyDefinitions [{
      boundBone {boneName=ribbon_v_02}
      boxExtents=<4 4 4>
      constraintSetup {angularLimitsMax=<15 60 45>, angularLimitsMin=<-10 -45 -30>, angularTarget=<0 0 0>}
      localJointOffset=<0.1 0 0>
     }]
    relativeSpaceBone {boneName=handle}
    simSpaceSettings {maxAngularAcceleration=50, maxAngularVelocity=1, simSpaceAngularAlpha=0.8}
   }
   Output Pose(Result=Component To Local(ComponentPose=@n86))
   Reroute Node(InputPin:Wildcard) -> (OutputPin:Wildcard)
  ]}
 EventGraph {nodes [
   Event BlueprintUpdateAnimation()
   > n1 = Sequence()
   > n2 = Branch(Condition=Get bIsSkyDiving())
   > n3 = Branch(Condition=Get bIsParachuteOpened())
   > Set GravityOverride(GravityOverride="0,0,500")
   Set GravityOverride(execute=@n3.else, GravityOverride="0,0,300")
   Set GravityOverride(execute=@n2.else, GravityOverride="0,0,-980")
   n4 = Branch(execute=@n1.then_1, Condition=Get bIsInFrontEnd())
   > Set Is In Front End(Is In Front End:bool=true) -> (Output_Get:bool)
   > Set Angular Spring Constant(Angular Spring Constant:double=500) -> (Output_Get:double)
   Set Not_FrontEnd(execute=@n4.else, Not_FrontEnd:bool=true) -> (Output_Get:bool)
   > Set Angular Spring Constant(Angular Spring Constant=1000)
   TryGetPawnOwner()
  ]}
]
variables [
 Angular Spring Constant {type=number}
 FrontEnd {type=boolean}
 GravityOverride {type=Vector}
 Is In Front End {type=boolean}
 Not_FrontEnd {type=boolean}
]
