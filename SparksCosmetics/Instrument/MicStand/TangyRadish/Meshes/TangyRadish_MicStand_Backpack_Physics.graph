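# Physics asset for the backpack mesh of the TangyRadish mic stand cosmetic, providing capsule bodies for the stand's handle and its decorative vines.
# The top, middle and bottom vines each constrain directly to the handle with small angular limit values, and the bottom and middle_02 groups lock their second swing axis.
# The three mic vines form a chain off the handle with much wider swing limits and locked twist.
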PhysicsAsset /SparksCosmetics/Instrument/MicStand/TangyRadish/Meshes/TangyRadish_MicStand_Backpack_Physics
bodies {
 bottom_vine_01 [{shapeType=Capsule, length=4, radius=4, shapeName=bottom_vine_01_capsule, center=<4 2 -4>, rotation=<rot 90 0 0>}]
 bottom_vine_02 [{shapeType=Capsule, length=4, radius=4, shapeName=bottom_vine_02_capsule, center=<4 0 0>, rotation=<rot 90 0 0>}]
 handle [{shapeType=Capsule, length=71.6544, radius=2, rotation=<rot 0 -90 0>}]
 mic_vine_01 [{shapeType=Capsule, length=2, radius=1, shapeName=mic_vine_01_capsule, center=<1.5 0 0>, rotation=<rot 0 -90 90>}]
 mic_vine_02 [{shapeType=Capsule, length=2, radius=1, shapeName=mic_vine_02_capsule, center=<1.5 0 0>, rotation=<rot 0 -90 90>}]
 mic_vine_03 [{shapeType=Capsule, length=2, radius=1, shapeName=mic_vine_03_capsule, center=<1.5 0 0>, rotation=<rot 0 -90 90>}]
 middle_02_vine_01 [{shapeType=Capsule, length=4, radius=4, shapeName=middle_02_vine_01_capsule, center=<5 0 -2>, rotation=<rot 90 0 0>}]
 middle_02_vine_02 [{shapeType=Capsule, length=4, radius=4, shapeName=middle_02_vine_02_capsule, center=<7 1 -3>, rotation=<rot 90 0 0>}]
 middle_vine_01 [{shapeType=Capsule, length=4, radius=4, shapeName=middle_vine_01_capsule, center=<-6.5 0.5 0>, rotation=<rot 90 0 0>}]
 middle_vine_02 [{shapeType=Capsule, length=4, radius=4, shapeName=middle_vine_02_capsule, center=<-1 1.5 2>, rotation=<rot 90 -30.5168 0>}]
 middle_vine_03 [{shapeType=Capsule, length=4, radius=4, shapeName=middle_vine_03_capsule, center=<-6 0 2.5>, rotation=<rot 90 0 0>}]
 top_vine_01 [{shapeType=Capsule, length=4, radius=4, shapeName=top_vine_01_capsule, center=<-2 0 2>, rotation=<rot 90 0 0>}]
 top_vine_02 [{shapeType=Capsule, length=4, radius=4, shapeName=top_vine_02_capsule, center=<-2 0 2>, rotation=<rot 90 0 0>}]
}
constraints [
 {bone1Name=top_vine_02, bone2Name=handle, swing1LimitDegrees=5, swing2LimitDegrees=2, twistLimitDegrees=2}
 {bone1Name=bottom_vine_01, bone2Name=handle, swing1LimitDegrees=5, swing2LimitDegrees=45, swing2Motion=Locked, twistLimitDegrees=2}
 {bone1Name=bottom_vine_02, bone2Name=handle, swing1LimitDegrees=5, swing2LimitDegrees=45, swing2Motion=Locked, twistLimitDegrees=2}
 {bone1Name=middle_02_vine_01, bone2Name=handle, swing1LimitDegrees=5, swing2LimitDegrees=6, swing2Motion=Locked, twistLimitDegrees=2}
 {bone1Name=middle_02_vine_02, bone2Name=handle, swing1LimitDegrees=5, swing2LimitDegrees=6, swing2Motion=Locked, twistLimitDegrees=2}
 {bone1Name=middle_vine_01, bone2Name=handle, swing1LimitDegrees=5, swing2LimitDegrees=2, twistLimitDegrees=2}
 {bone1Name=middle_vine_02, bone2Name=handle, swing1LimitDegrees=5, swing2LimitDegrees=2, twistLimitDegrees=2}
 {bone1Name=middle_vine_03, bone2Name=handle, swing1LimitDegrees=5, swing2LimitDegrees=2, twistLimitDegrees=2}
 {bone1Name=top_vine_01, bone2Name=handle, swing1LimitDegrees=5, swing2LimitDegrees=2, twistLimitDegrees=2}
 {bone1Name=mic_vine_01, bone2Name=handle, swing1LimitDegrees=45, swing2LimitDegrees=35, twistLimitDegrees=45, twistMotion=Locked}
 {bone1Name=mic_vine_02, bone2Name=mic_vine_01, swing1LimitDegrees=60, swing2LimitDegrees=60, twistLimitDegrees=45, twistMotion=Locked}
 {bone1Name=mic_vine_03, bone2Name=mic_vine_02, swing1LimitDegrees=45, swing2LimitDegrees=60, twistLimitDegrees=45, twistMotion=Locked}
]
