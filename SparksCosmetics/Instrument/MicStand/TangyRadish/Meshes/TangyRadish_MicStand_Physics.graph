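# Physics asset for the TangyRadish mic stand instrument cosmetic, giving its handle and decorative vines simple collision bodies for physical motion.
# A long capsule covers the handle, and nine short capsules cover the top, middle, middle_02 and bottom vine bones.
# Each vine constrains directly to the handle rather than chaining, with tight swing and twist limits for slight wobble.
# The bottom and middle_02 vines also lock their second swing axis.
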PhysicsAsset /SparksCosmetics/Instrument/MicStand/TangyRadish/Meshes/TangyRadish_MicStand_Physics
bodies {
 bottom_vine_01 [{shapeType=Capsule, length=5.0343, radius=5.4716, shapeName=bottom_vine_01_capsule, center=<4 2 -4>, rotation=<rot 90 0 0>}]
 bottom_vine_02 [{shapeType=Capsule, length=5.478, radius=4.9395, shapeName=bottom_vine_02_capsule, center=<4 0 0>, rotation=<rot 90 0 0>}]
 handle [{shapeType=Capsule, length=71.6544, radius=2, rotation=<rot 0 -90 0>}]
 middle_02_vine_01 [{shapeType=Capsule, length=4.987, radius=5.2075, shapeName=middle_02_vine_01_capsule, center=<5 0 -2>, rotation=<rot 90 0 0>}]
 middle_02_vine_02 [{shapeType=Capsule, length=5.8604, radius=5.0416, shapeName=middle_02_vine_02_capsule, center=<7 1 -3>, rotation=<rot 90 0 0>}]
 middle_vine_01 [{shapeType=Capsule, length=3.7727, radius=5.2813, shapeName=middle_vine_01_capsule, center=<-6.5 0.5 0>, rotation=<rot 90 0 0>}]
 middle_vine_02 [{shapeType=Capsule, length=1.8362, radius=3.2764, shapeName=middle_vine_02_capsule, center=<-1 1.5 2>, rotation=<rot 90 -30.5168 0>}]
 middle_vine_03 [{shapeType=Capsule, length=4.9607, radius=4.8455, shapeName=middle_vine_03_capsule, center=<-6 0 2.5>, rotation=<rot 90 0 0>}]
 top_vine_01 [{shapeType=Capsule, length=1.7541, radius=4.3385, shapeName=top_vine_01_capsule, center=<-2 0 2>, rotation=<rot 90 0 0>}]
 top_vine_02 [{shapeType=Capsule, length=2.0201, radius=3.95, shapeName=top_vine_02_capsule, center=<-2 0 2>, rotation=<rot 90 0 0>}]
}
constraints [
 {bone1Name=middle_vine_03, bone2Name=handle, swing1LimitDegrees=5, swing2LimitDegrees=2, twistLimitDegrees=2}
 {bone1Name=middle_vine_01, bone2Name=handle, swing1LimitDegrees=5, swing2LimitDegrees=2, twistLimitDegrees=2}
 {bone1Name=middle_vine_02, bone2Name=handle, swing1LimitDegrees=5, swing2LimitDegrees=2, twistLimitDegrees=2}
 {bone1Name=top_vine_02, bone2Name=handle, swing1LimitDegrees=5, swing2LimitDegrees=2, twistLimitDegrees=2}
 {bone1Name=top_vine_01, bone2Name=handle, swing1LimitDegrees=5, swing2LimitDegrees=2, twistLimitDegrees=2}
 {bone1Name=bottom_vine_02, bone2Name=handle, swing1LimitDegrees=5, swing2LimitDegrees=5, swing2Motion=Locked, twistLimitDegrees=2}
 {bone1Name=bottom_vine_01, bone2Name=handle, swing1LimitDegrees=5, swing2LimitDegrees=5, swing2Motion=Locked, twistLimitDegrees=2}
 {bone1Name=middle_02_vine_02, bone2Name=handle, swing1LimitDegrees=5, swing2LimitDegrees=6, swing2Motion=Locked, twistLimitDegrees=2}
 {bone1Name=middle_02_vine_01, bone2Name=handle, swing1LimitDegrees=5, swing2LimitDegrees=6, swing2Motion=Locked, twistLimitDegrees=2}
]
