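# Animation Blueprint for the TangyRadish mic stand cosmetic, adding physics-driven secondary motion to the incoming pose.
# Two RigidBody nodes run in series in component space, each simulating relative to the handle bone with its own physics asset and damped component motion.
# The first, using the stand physics, has its alpha bool fixed off; the second, using the backpack physics, runs only when bIsInFrontEnd is false.
# The event graph holds only unconnected default nodes.
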AnimBlueprint /SparksCosmetics/Instrument/MicStand/TangyRadish/Meshes/TangyRadish_MicStand_AnimBP
target_skeleton=&/TangyRadish_MicStand_Skeleton
graphs [
 AnimGraph {nodes [
   n1 = RigidBody!(
    ComponentPose=Local To Component(LocalPose=Input Pose())
    bAlphaBoolEnabled:bool=false
    OverrideWorldGravity:Vector="0,0,-980"
    ExternalForce
   ) -> (Pose) {
    alphaInputType=Bool
    bOverrideWorldGravity=true
    overridePhysicsAsset=&/TangyRadish_MicStand_Physics
    simulationSpace=BaseBoneSpace
    alphaBoolBlend {blendOption=Linear}
    baseBoneRef {boneName=handle}
    componentAppliedLinearAccClamp=<1000 1000 900>
    componentLinearAccScale=<0.35 0.3 0.2>
    componentLinearVelScale=<0.35 0.3 0.2>
    overrideWorldGravity {z=-980}
    simSpaceSettings {maxAngularAcceleration=600, maxAngularVelocity=3, maxLinearAcceleration=700, maxLinearVelocity=500, externalLinearDragV=<0.2 0.2>}
   }
   n2 = RigidBody(ComponentPose=@n1, bAlphaBoolEnabled=NOT Boolean(A=Get bIsInFrontEnd()), OverrideWorldGravity="0,0,-980") {
    alphaInputType=Bool
    bOverrideWorldGravity=true
    overridePhysicsAsset=&/TangyRadish_MicStand_Backpack_Physics
    simulationSpace=BaseBoneSpace
    alphaBoolBlend {blendOption=Linear}
    baseBoneRef {boneName=handle}
    componentAppliedLinearAccClamp=<1000 1000 900>
    componentLinearAccScale=<0.35 0.3 0.2>
    componentLinearVelScale=<0.35 0.3 0.2>
    overrideWorldGravity {z=-980}
    simSpaceSettings {
     dampingAlpha=0.5
     maxAngularAcceleration=200
     maxAngularVelocity=3
     maxLinearAcceleration=700
     maxLinearVelocity=500
     velocityScaleZ=0.4
     worldAlpha=0.6
     externalLinearDragV=<0.2 0.2>
    }
   }
   Output Pose(Result=Component To Local(ComponentPose=@n2))
  ]}
 EventGraph {nodes [
   TryGetPawnOwner()
   Event BlueprintUpdateAnimation()
  ]}
]
variables []
